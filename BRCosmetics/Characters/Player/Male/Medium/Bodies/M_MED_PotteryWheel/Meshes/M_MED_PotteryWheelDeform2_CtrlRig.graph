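# Control Rig that adds secondary and corrective bone deformation for the M_MED_PotteryWheel male player body.
# Position constraints move chest, trap, clavicle-aim and inner-thigh bones, then spherical pose readers on legs, clavicles and arms weight transform offsets on muscle and joint helper bones.
# Swing-twist extraction counter-rotates upper-arm twist bones (right 02 and 03 receive identity instead) and feeds forearm twist from half-weighted hand-split rotations.
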ControlRigBlueprint /BRCosmetics/Characters/Player/Male/Medium/Bodies/M_MED_PotteryWheel/Meshes/M_MED_PotteryWheelDeform2_CtrlRig
graphs [
 RigVMFunctionLibrary {links [], nodes []}
 RigVMModel {nodes [
   RigUnit_BeginExecution = &RigUnit_BeginExecution()
   /Script/ControlRig.RigUnit_PositionConstraintLocalSpaceOffset(
    Child=(Type=Bone,Name="chest_l")
    Filter=(bX=True,bY=True,bZ=True)
    Parents=((Item=(Type=Bone,Name="deform_upperarm_l"),Weight=1),(Item=(Type=Bone,Name="deform_up_arm_twist_02_l"),Weight=1),(Item=(Type=Bone,Name="spine_04"),Weight=1))
    Weight=1
    bMaintainOffset=True
    ExecutePin=@RigUnit_BeginExecution.ExecutePin
   ) {is=struct RigVMUnitNode}
   > &RigUnit_PositionConstraintLocalSpaceOffset(
    Child=(Type=Bone,Name="chest_r")
    Filter=(bX=True,bY=True,bZ=True)
    Parents=((Item=(Type=Bone,Name="upperarm_retarg_r"),Weight=1),(Item=(Type=Bone,Name="deform_up_arm_twist_02_r"),Weight=1),(Item=(Type=Bone,Name="spine_04"),Weight=1))
    Weight=1
    bMaintainOffset=True
   )
   > &RigUnit_PositionConstraintLocalSpaceOffset(
    Child=(Type=Bone,Name="trap_l")
    Filter=(bX=True,bY=True,bZ=True)
    Parents=((Item=(Type=Bone,Name="deform_up_arm_twist_01_l"),Weight=1),(Item=(Type=Bone,Name="head"),Weight=1))
    Weight=1
    bMaintainOffset=True
   )
   > &RigUnit_PositionConstraintLocalSpaceOffset(
    Child=(Type=Bone,Name="trap_r")
    Filter=(bX=True,bY=True,bZ=True)
    Parents=((Item=(Type=Bone,Name="deform_up_arm_twist_01_r"),Weight=1),(Item=(Type=Bone,Name="head"),Weight=1))
    Weight=1
    bMaintainOffset=True
   )
   > &RigUnit_PositionConstraintLocalSpaceOffset(
    Child=(Type=Bone,Name="deform_clavicle_aim_l")
    Filter=(bX=True,bY=True,bZ=True)
    Parents=((Item=(Type=Bone,Name="deform_up_arm_twist_01_l"),Weight=1),(Item=(Type=Bone,Name="deform_upperarm_l"),Weight=1))
    Weight=1
    bMaintainOffset=True
   )
   > &RigUnit_PositionConstraintLocalSpaceOffset(
    Child=(Type=Bone,Name="clavicle_aim_retarget_r")
    Filter=(bX=True,bY=True,bZ=True)
    Parents=((Item=(Type=Bone,Name="deform_up_arm_twist_01_r"),Weight=1),(Item=(Type=Bone,Name="deform_upperarm_r"),Weight=1))
    Weight=1
    bMaintainOffset=True
   )
   > &RigUnit_PositionConstraintLocalSpaceOffset(
    Child=(Type=Bone,Name="thigh_in_l")
    Filter=(bX=True,bY=True,bZ=True)
    Parents=((Item=(Type=Bone,Name="thigh_twist_01_l"),Weight=1))
    Weight=1
    bMaintainOffset=True
   )
   > &RigUnit_PositionConstraintLocalSpaceOffset(
    Child=(Type=Bone,Name="thigh_in_r")
    Filter=(bX=True,bY=True,bZ=True)
    Parents=((Item=(Type=Bone,Name="thigh_twist_01_r"),Weight=1))
    Weight=1
    bMaintainOffset=True
   )
   > n1 = &RigUnit_SphericalPoseReader(
    ActiveRegionScaleFactors=(PositiveWidth=0.1,NegativeWidth=0.1,PositiveHeight=0.1,NegativeHeight=0.1)
    ActiveRegionSize=0.1
    Debug=(bDrawDebug=false,bDraw2D=false,DebugScale=25,DebugSegments=20,DebugThickness=0.25)
    DriverAxis=(X=-1,Y=0,Z=0)
    DriverItem=(Type=Bone,Name="thigh_l")
    FalloffRegionScaleFactors=(PositiveWidth=0,NegativeWidth=2,PositiveHeight=2,NegativeHeight=2)
    FalloffSize=0.45
    OptionalParentItem=(Type=Bone,Name="pelvis")
    RotationOffset=(X=90,Y=0,Z=-90)
   )
   > &RigUnit_ModifyTransforms(
    ItemToModify=((Item=(Type=Bone,Name="thigh_ot_l"),Transform=(Rotation=(X=0,Y=0,Z=0.1736,W=0.9848),Translation=(X=-10,Y=5,Z=5),Scale3D=(X=1,Y=1,Z=1))))
    WeightMinimum=0
    Weight=@n1.OutputParam
   )
   > n2 = &RigUnit_SphericalPoseReader(
    ActiveRegionScaleFactors=(PositiveWidth=0.1,NegativeWidth=0.1,PositiveHeight=0.1,NegativeHeight=0.1)
    ActiveRegionSize=0.1
    Debug=(bDrawDebug=false,bDraw2D=false,DebugScale=25,DebugSegments=20,DebugThickness=0.25)
    DriverAxis=(X=1,Y=0,Z=0)
    DriverItem=(Type=Bone,Name="thigh_r")
    FalloffRegionScaleFactors=(PositiveWidth=0,NegativeWidth=2,PositiveHeight=2,NegativeHeight=2)
    FalloffSize=0.45
    OptionalParentItem=(Type=Bone,Name="pelvis")
    RotationOffset=(X=90,Y=0,Z=90)
   )
   > &RigUnit_ModifyTransforms(
    ItemToModify=((Item=(Type=Bone,Name="thigh_ot_r"),Transform=(Rotation=(X=0,Y=0,Z=0,W=1),Translation=(X=10,Y=-5,Z=-5),Scale3D=(X=1,Y=1,Z=1))))
    WeightMinimum=0
    Weight=@n2.OutputParam
   )
   > n3 = &RigUnit_SphericalPoseReader(
    ActiveRegionScaleFactors=(PositiveWidth=0.1,NegativeWidth=0.1,PositiveHeight=0.1,NegativeHeight=0.1)
    ActiveRegionSize=0.1
    Debug=(bDrawDebug=false,bDraw2D=false,DebugScale=25,DebugSegments=20,DebugThickness=0.25)
    DriverAxis=(X=-1,Y=0,Z=0)
    DriverItem=(Type=Bone,Name="calf_l")
    FalloffRegionScaleFactors=(PositiveWidth=0,NegativeWidth=1.5,PositiveHeight=1.5,NegativeHeight=1.5)
    FalloffSize=0.45
    OptionalParentItem=(Type=Bone,Name="thigh_l")
    RotationOffset=(X=-90,Y=0,Z=90)
   )
   > &RigUnit_ModifyTransforms(
    ItemToModify=((Item=(Type=Bone,Name="calf_ot_l"),Transform=(Rotation=(X=0,Y=0,Z=0,W=1),Translation=(X=5,Y=-10,Z=0),Scale3D=(X=1,Y=1,Z=1))))
    WeightMinimum=0
    Weight=@n3.OutputParam
   )
   > &RigUnit_ModifyTransforms(
    ItemToModify=((Item=(Type=Bone,Name="calf_in_l"),Transform=(Rotation=(X=0,Y=0,Z=-0.5,W=0.866),Translation=(X=-20,Y=-20,Z=0),Scale3D=(X=1,Y=1,Z=1))))
    WeightMinimum=0
    Weight=@n3.OutputParam
   )
   > &RigUnit_ModifyTransforms(
    ItemToModify=((Item=(Type=Bone,Name="calf_counter_l"),Transform=(Rotation=(X=0,Y=0,Z=-0.0872,W=0.9962),Translation=(X=0,Y=-2,Z=0),Scale3D=(X=1,Y=1,Z=1))))
    WeightMinimum=0
    Weight=@n3.OutputParam
   )
   > &RigUnit_ModifyTransforms(
    ItemToModify=((Item=(Type=Bone,Name="thigh_counter_l"),Transform=(Rotation=(X=0,Y=0,Z=0.1305,W=0.9914),Translation=(X=0,Y=0,Z=0),Scale3D=(X=1,Y=1,Z=1))))
    WeightMinimum=0
    Weight=@n3.OutputParam
   )
   > &RigUnit_ModifyTransforms(
    ItemToModify=((Item=(Type=Bone,Name="calf_kneepad_l"),Transform=(Rotation=(X=0,Y=0,Z=-0.2164,W=0.9763),Translation=(X=0,Y=0,Z=0),Scale3D=(X=1,Y=1,Z=1))))
    WeightMinimum=0
    Weight=@n3.OutputParam
   )
   > n4 = &RigUnit_SphericalPoseReader(
    ActiveRegionScaleFactors=(PositiveWidth=0.1,NegativeWidth=0.1,PositiveHeight=0.1,NegativeHeight=0.1)
    ActiveRegionSize=0.1
    Debug=(bDrawDebug=false,bDraw2D=false,DebugScale=25,DebugSegments=20,DebugThickness=0.25)
    DriverAxis=(X=1,Y=0,Z=0)
    DriverItem=(Type=Bone,Name="calf_r")
    FalloffRegionScaleFactors=(PositiveWidth=0,NegativeWidth=1.5,PositiveHeight=1.5,NegativeHeight=1.5)
    FalloffSize=0.45
    OptionalParentItem=(Type=Bone,Name="thigh_r")
    RotationOffset=(X=-90,Y=0,Z=-90)
   )
   > &RigUnit_ModifyTransforms(
    ItemToModify=((Item=(Type=Bone,Name="calf_ot_r"),Transform=(Rotation=(X=0,Y=0,Z=0,W=1),Translation=(X=-5,Y=10,Z=0),Scale3D=(X=1,Y=1,Z=1))))
    WeightMinimum=0
    Weight=@n4.OutputParam
   )
   > &RigUnit_ModifyTransforms(
    ItemToModify=((Item=(Type=Bone,Name="calf_in_r"),Transform=(Rotation=(X=0,Y=0,Z=0.5,W=0.866),Translation=(X=20,Y=20,Z=0),Scale3D=(X=1,Y=1,Z=1))))
    WeightMinimum=0
    Weight=@n4.OutputParam
   )
   > &RigUnit_ModifyTransforms(
    ItemToModify=((Item=(Type=Bone,Name="calf_counter_r"),Transform=(Rotation=(X=0,Y=0,Z=-0.0436,W=0.999),Translation=(X=0,Y=0,Z=0),Scale3D=(X=1,Y=1,Z=1))))
    WeightMinimum=0
    Weight=@n4.OutputParam
   )
   > &RigUnit_ModifyTransforms(
    ItemToModify=((Item=(Type=Bone,Name="thigh_counter_r"),Transform=(Rotation=(X=0,Y=0,Z=0.0436,W=0.999),Translation=(X=0,Y=0,Z=0),Scale3D=(X=1,Y=1,Z=1))))
    WeightMinimum=0
    Weight=@n4.OutputParam
   )
   > &RigUnit_ModifyTransforms(
    ItemToModify=((Item=(Type=Bone,Name="calf_kneepad_r"),Transform=(Rotation=(X=0,Y=0,Z=-0.2164,W=0.9763),Translation=(X=0,Y=0,Z=0),Scale3D=(X=1,Y=1,Z=1))))
    WeightMinimum=0
    Weight=@n4.OutputParam
   )
   > n5 = &RigUnit_SphericalPoseReader(
    ActiveRegionScaleFactors=(PositiveWidth=0.1,NegativeWidth=0.1,PositiveHeight=0.1,NegativeHeight=0.1)
    ActiveRegionSize=0.1
    Debug=(bDrawDebug=false,bDraw2D=false,DebugScale=25,DebugSegments=20,DebugThickness=0.25)
    DriverAxis=(X=-1,Y=0,Z=0)
    DriverItem=(Type=Bone,Name="calf_l")
    FalloffRegionScaleFactors=(PositiveWidth=0,NegativeWidth=0.7,PositiveHeight=0.7,NegativeHeight=0.7)
    FalloffSize=0.45
    OptionalParentItem=(Type=Bone,Name="thigh_l")
    RotationOffset=(X=-90,Y=0,Z=90)
   )
   > &RigUnit_ModifyTransforms(
    ItemToModify=((Item=(Type=Bone,Name="calf_counter_l"),Transform=(Rotation=(X=0,Y=0,Z=-0.2588,W=0.9659),Translation=(X=0,Y=0,Z=0),Scale3D=(X=1,Y=1,Z=1))))
    WeightMinimum=0
    Weight=@n5.OutputParam
   )
   > &RigUnit_ModifyTransforms(
    ItemToModify=((Item=(Type=Bone,Name="thigh_counter_l"),Transform=(Rotation=(X=0,Y=0,Z=0.2588,W=0.9659),Translation=(X=0,Y=0,Z=0),Scale3D=(X=1,Y=1,Z=1.5))))
    WeightMinimum=0
    Weight=@n5.OutputParam
   )
   > &RigUnit_ModifyTransforms(
    ItemToModify=((Item=(Type=Bone,Name="calf_mscl_l"),Transform=(Rotation=(X=0,Y=0,Z=0,W=1),Translation=(X=0,Y=0,Z=0),Scale3D=(X=1,Y=0.5,Z=1.5))))
    WeightMinimum=0
    Weight=@n5.OutputParam
   )
   > &RigUnit_ModifyTransforms(
    ItemToModify=((Item=(Type=Bone,Name="calf_twist_01_l"),Transform=(Rotation=(X=0,Y=0,Z=0,W=1),Translation=(X=0,Y=-2,Z=0),Scale3D=(X=1,Y=1,Z=1))))
    WeightMinimum=0
    Weight=@n5.OutputParam
   )
   > &RigUnit_ModifyTransforms(
    ItemToModify=((Item=(Type=Bone,Name="calf_kneepad_l"),Transform=(Rotation=(X=0,Y=0,Z=0.0872,W=0.9962),Translation=(X=0,Y=0,Z=0),Scale3D=(X=1,Y=1,Z=1))))
    WeightMinimum=0
    Weight=@n5.OutputParam
   )
   > n6 = &RigUnit_SphericalPoseReader(
    ActiveRegionScaleFactors=(PositiveWidth=0.1,NegativeWidth=0.1,PositiveHeight=0.1,NegativeHeight=0.1)
    ActiveRegionSize=0.1
    Debug=(bDrawDebug=false,bDraw2D=false,DebugScale=25,DebugSegments=20,DebugThickness=0.25)
    DriverAxis=(X=1,Y=0,Z=0)
    DriverItem=(Type=Bone,Name="calf_r")
    FalloffRegionScaleFactors=(PositiveWidth=0,NegativeWidth=0.7,PositiveHeight=0.7,NegativeHeight=0.7)
    FalloffSize=0.45
    OptionalParentItem=(Type=Bone,Name="thigh_r")
    RotationOffset=(X=-90,Y=0,Z=-90)
   )
   > &RigUnit_ModifyTransforms(
    ItemToModify=((Item=(Type=Bone,Name="calf_counter_r"),Transform=(Rotation=(X=0,Y=0,Z=-0.2588,W=0.9659),Translation=(X=0,Y=0,Z=0),Scale3D=(X=1,Y=1,Z=1))))
    WeightMinimum=0
    Weight=@n6.OutputParam
   )
   > &RigUnit_ModifyTransforms(
    ItemToModify=((Item=(Type=Bone,Name="thigh_counter_r"),Transform=(Rotation=(X=0,Y=0,Z=0.2588,W=0.9659),Translation=(X=0,Y=-10,Z=0),Scale3D=(X=1,Y=1,Z=1))))
    WeightMinimum=0
    Weight=@n6.OutputParam
   )
   > &RigUnit_ModifyTransforms(
    ItemToModify=((Item=(Type=Bone,Name="calf_mscl_r"),Transform=(Rotation=(X=0,Y=0,Z=0,W=1),Translation=(X=0,Y=0,Z=0),Scale3D=(X=1,Y=0.5,Z=1.5))))
    WeightMinimum=0
    Weight=@n6.OutputParam
   )
   > &RigUnit_ModifyTransforms(
    ItemToModify=((Item=(Type=Bone,Name="calf_twist_01_r"),Transform=(Rotation=(X=0,Y=0,Z=0,W=1),Translation=(X=0,Y=-2,Z=0),Scale3D=(X=1,Y=1,Z=1))))
    WeightMinimum=0
    Weight=@n6.OutputParam
   )
   > &RigUnit_ModifyTransforms(
    ItemToModify=((Item=(Type=Bone,Name="calf_kneepad_r"),Transform=(Rotation=(X=0,Y=0,Z=0.0872,W=0.9962),Translation=(X=0,Y=0,Z=0),Scale3D=(X=1,Y=1,Z=1))))
    WeightMinimum=0
    Weight=@n6.OutputParam
   )
   > n7 = &RigUnit_SphericalPoseReader(
    ActiveRegionScaleFactors=(PositiveWidth=0.1,NegativeWidth=3,PositiveHeight=3,NegativeHeight=3)
    ActiveRegionSize=0.1
    Debug=(bDrawDebug=false,bDraw2D=false,DebugScale=25,DebugSegments=20,DebugThickness=0.25)
    DriverAxis=(X=1,Y=0,Z=0)
    DriverItem=(Type=Bone,Name="deform_lowerarm_l")
    FalloffRegionScaleFactors=(PositiveWidth=1,NegativeWidth=1,PositiveHeight=1,NegativeHeight=1)
    FalloffSize=0.45
    OptionalParentItem=(Type=Bone,Name="spine_05")
    RotationOffset=(X=90,Y=0,Z=0)
   )
   > &RigUnit_ModifyTransforms(
    ItemToModify=((Item=(Type=Bone,Name="deform_lowerarm_in_l"),Transform=(Rotation=(X=0,Y=0,Z=0,W=1),Translation=(X=3,Y=3,Z=0),Scale3D=(X=1,Y=1,Z=1))))
    WeightMinimum=0
    Weight=@n7.OutputParam
   )
   > &RigUnit_ModifyTransforms(
    ItemToModify=((Item=(Type=Bone,Name="deform_lowerarm_ot_l"),Transform=(Rotation=(X=0,Y=0,Z=0,W=1),Translation=(X=-3,Y=-3,Z=0),Scale3D=(X=1,Y=1,Z=1))))
    WeightMinimum=0
    Weight=@n7.OutputParam
   )
   > &RigUnit_ModifyTransforms(
    ItemToModify=((Item=(Type=Bone,Name="deform_bicep_03_l"),Transform=(Rotation=(X=0,Y=0,Z=0,W=1),Translation=(X=0,Y=2,Z=0),Scale3D=(X=1,Y=1,Z=1))))
    WeightMinimum=0
    Weight=@n7.OutputParam
   )
   > &RigUnit_ModifyTransforms(
    ItemToModify=((Item=(Type=Bone,Name="deform_bicep_04_l"),Transform=(Rotation=(X=0,Y=0,Z=0,W=1),Translation=(X=0,Y=1,Z=0),Scale3D=(X=1,Y=1,Z=1))))
    WeightMinimum=0
    Weight=@n7.OutputParam
   )
   > &RigUnit_ModifyTransforms(
    ItemToModify=((Item=(Type=Bone,Name="deform_tricep_03_l"),Transform=(Rotation=(X=0,Y=0,Z=0,W=1),Translation=(X=3,Y=0,Z=0),Scale3D=(X=1,Y=1,Z=1))))
    WeightMinimum=0
    Weight=@n7.OutputParam
   )
   > n8 = &RigUnit_SphericalPoseReader(
    ActiveRegionScaleFactors=(PositiveWidth=0.1,NegativeWidth=3,PositiveHeight=3,NegativeHeight=3)
    ActiveRegionSize=0.1
    Debug=(bDrawDebug=false,bDraw2D=false,DebugScale=25,DebugSegments=20,DebugThickness=0.25)
    DriverAxis=(X=-1,Y=0,Z=0)
    DriverItem=(Type=Bone,Name="deform_lowerarm_r")
    FalloffRegionScaleFactors=(PositiveWidth=1,NegativeWidth=1,PositiveHeight=1,NegativeHeight=1)
    FalloffSize=0.45
    OptionalParentItem=(Type=Bone,Name="spine_05")
    RotationOffset=(X=90,Y=0,Z=180)
   )
   > &RigUnit_ModifyTransforms(
    ItemToModify=((Item=(Type=Bone,Name="deform_lowerarm_in_r"),Transform=(Rotation=(X=0,Y=0,Z=0,W=1),Translation=(X=-3,Y=-3,Z=0),Scale3D=(X=1,Y=1,Z=1))))
    WeightMinimum=0
    Weight=@n8.OutputParam
   )
   > &RigUnit_ModifyTransforms(
    ItemToModify=((Item=(Type=Bone,Name="deform_lowerarm_ot_r"),Transform=(Rotation=(X=0,Y=0,Z=0,W=1),Translation=(X=3,Y=3,Z=0),Scale3D=(X=1,Y=1,Z=1))))
    WeightMinimum=0
    Weight=@n8.OutputParam
   )
   > &RigUnit_ModifyTransforms(
    ItemToModify=((Item=(Type=Bone,Name="deform_bicep_03_r"),Transform=(Rotation=(X=0,Y=0,Z=0,W=1),Translation=(X=0,Y=-2,Z=0),Scale3D=(X=1,Y=1,Z=1))))
    WeightMaximum=0
    WeightMinimum=0
    Weight=@n8.OutputParam
   )
   > &RigUnit_ModifyTransforms(
    ItemToModify=((Item=(Type=Bone,Name="deform_bicep_04_r"),Transform=(Rotation=(X=0,Y=0,Z=0,W=1),Translation=(X=0,Y=-1,Z=0),Scale3D=(X=1,Y=1,Z=1))))
    WeightMinimum=0
    Weight=@n8.OutputParam
   )
   > &RigUnit_ModifyTransforms(
    ItemToModify=((Item=(Type=Bone,Name="deform_tricep_04_r"),Transform=(Rotation=(X=0,Y=0,Z=0,W=1),Translation=(X=-3,Y=0,Z=0),Scale3D=(X=1,Y=1,Z=1))))
    WeightMinimum=0
    Weight=@n8.OutputParam
   )
   > n9 = &RigUnit_SphericalPoseReader(
    ActiveRegionScaleFactors=(PositiveWidth=0.1,NegativeWidth=0.1,PositiveHeight=0.1,NegativeHeight=0.1)
    ActiveRegionSize=0.1
    Debug=(bDrawDebug=false,bDraw2D=false,DebugScale=25,DebugSegments=20,DebugThickness=0.25)
    DriverAxis=(X=1,Y=0,Z=0)
    DriverItem=(Type=Bone,Name="deform_clavicle_l")
    FalloffRegionScaleFactors=(PositiveWidth=1,NegativeWidth=1,PositiveHeight=1,NegativeHeight=1)
    FalloffSize=0.45
    OptionalParentItem=(Type=Bone,Name="spine_05")
    RotationOffset=(X=0,Y=0,Z=0)
   )
   > &RigUnit_ModifyTransforms(
    ItemToModify=((Item=(Type=Bone,Name="lat_l"),Transform=(Rotation=(X=0,Y=0,Z=0,W=1),Translation=(X=25,Y=10,Z=0),Scale3D=(X=1,Y=1,Z=1))))
    WeightMinimum=0
    Weight=@n9.OutputParam
   )
   > &RigUnit_ModifyTransforms(
    ItemToModify=((Item=(Type=Bone,Name="clavicle_aim_retarget_l"),Transform=(Rotation=(X=0,Y=-0.5,Z=0,W=0.866),Translation=(X=10,Y=-10,Z=0),Scale3D=(X=1,Y=1,Z=1))))
    WeightMinimum=0
    Weight=@n9.OutputParam
   )
   > &RigUnit_ModifyTransforms(
    ItemToModify=((Item=(Type=Bone,Name="chest_l"),Transform=(Rotation=(X=-0.342,Y=0,Z=0,W=0.9397),Translation=(X=0,Y=-3,Z=0),Scale3D=(X=1,Y=1,Z=1))))
    WeightMinimum=0
    Weight=@n9.OutputParam
   )
   > n10 = &RigUnit_SphericalPoseReader(
    ActiveRegionScaleFactors=(PositiveWidth=0.1,NegativeWidth=0.1,PositiveHeight=0.1,NegativeHeight=0.1)
    ActiveRegionSize=0.1
    Debug=(bDrawDebug=false,bDraw2D=false,DebugScale=25,DebugSegments=20,DebugThickness=0.25)
    DriverAxis=(X=-1,Y=0,Z=0)
    DriverItem=(Type=Bone,Name="deform_clavicle_r")
    FalloffRegionScaleFactors=(PositiveWidth=1,NegativeWidth=0,PositiveHeight=1,NegativeHeight=1)
    FalloffSize=0.45
    OptionalParentItem=(Type=Bone,Name="spine_05")
    RotationOffset=(X=0,Y=180,Z=0)
   )
   > &RigUnit_ModifyTransforms(
    ItemToModify=((Item=(Type=Bone,Name="lat_r"),Transform=(Rotation=(X=0,Y=0,Z=0,W=1),Translation=(X=-25,Y=-10,Z=0),Scale3D=(X=1,Y=1,Z=1))))
    WeightMinimum=0
    Weight=@n10.OutputParam
   )
   > &RigUnit_ModifyTransforms(
    ItemToModify=((Item=(Type=Bone,Name="deform_clavicle_aim_r"),Transform=(Rotation=(X=0,Y=-0.5,Z=0,W=0.866),Translation=(X=-10,Y=10,Z=0),Scale3D=(X=1,Y=1,Z=1))))
    WeightMinimum=0
    Weight=@n10.OutputParam
   )
   > &RigUnit_ModifyTransforms(
    ItemToModify=((Item=(Type=Bone,Name="chest_r"),Transform=(Rotation=(X=-0.342,Y=0,Z=0,W=0.9397),Translation=(X=0,Y=3,Z=0),Scale3D=(X=1,Y=1,Z=1))))
    WeightMinimum=0
    Weight=@n10.OutputParam
   )
   > n11 = &RigUnit_SphericalPoseReader(
    ActiveRegionScaleFactors=(PositiveWidth=0.1,NegativeWidth=0.1,PositiveHeight=0.1,NegativeHeight=0.1)
    ActiveRegionSize=0.1
    Debug=(bDrawDebug=false,bDraw2D=false,DebugScale=25,DebugSegments=20,DebugThickness=0.25)
    DriverAxis=(X=1,Y=0,Z=0)
    DriverItem=(Type=Bone,Name="deform_clavicle_l")
    FalloffRegionScaleFactors=(PositiveWidth=1,NegativeWidth=0.5,PositiveHeight=1,NegativeHeight=1)
    FalloffSize=0.45
    OptionalParentItem=(Type=Bone,Name="spine_05")
    RotationOffset=(X=-90,Y=0,Z=0)
   )
   > &RigUnit_ModifyTransforms(
    ItemToModify=((Item=(Type=Bone,Name="trap_l"),Transform=(Rotation=(X=0,Y=0,Z=0,W=1),Translation=(X=0,Y=0,Z=0),Scale3D=(X=1,Y=1,Z=1))))
    WeightMinimum=0
    Weight=@n11.OutputParam
   )
   > &RigUnit_ModifyTransforms(
    ItemToModify=((Item=(Type=Bone,Name="lat_l"),Transform=(Rotation=(X=0,Y=0,Z=0,W=1),Translation=(X=-15,Y=5,Z=0),Scale3D=(X=1,Y=1,Z=1))))
    WeightMinimum=0
    Weight=@n11.OutputParam
   )
   > n12 = &RigUnit_SphericalPoseReader(
    ActiveRegionScaleFactors=(PositiveWidth=0.1,NegativeWidth=0.1,PositiveHeight=0.1,NegativeHeight=0.1)
    ActiveRegionSize=0.1
    Debug=(bDrawDebug=false,bDraw2D=false,DebugScale=25,DebugSegments=20,DebugThickness=0.25)
    DriverAxis=(X=-1,Y=0,Z=0)
    DriverItem=(Type=Bone,Name="deform_clavicle_r")
    FalloffRegionScaleFactors=(PositiveWidth=1,NegativeWidth=0,PositiveHeight=1,NegativeHeight=1)
    FalloffSize=0.45
    OptionalParentItem=(Type=Bone,Name="spine_05")
    RotationOffset=(X=-90,Y=0,Z=180)
   )
   > &RigUnit_ModifyTransforms(
    ItemToModify=((Item=(Type=Bone,Name="trap_l"),Transform=(Rotation=(X=0,Y=0,Z=0,W=1),Translation=(X=0,Y=0,Z=0),Scale3D=(X=1,Y=1,Z=1))))
    WeightMinimum=0
    Weight=@n12.OutputParam
   )
   > &RigUnit_ModifyTransforms(
    ItemToModify=((Item=(Type=Bone,Name="lat_l"),Transform=(Rotation=(X=0,Y=0,Z=0,W=1),Translation=(X=15,Y=-5,Z=0),Scale3D=(X=1,Y=1,Z=1))))
    WeightMinimum=0
    Weight=@n12.OutputParam
   )
   > n13 = &RigUnit_SphericalPoseReader(
    ActiveRegionScaleFactors=(PositiveWidth=0.1,NegativeWidth=0.1,PositiveHeight=0.1,NegativeHeight=0.1)
    ActiveRegionSize=0.1
    Debug=(bDrawDebug=false,bDraw2D=false,DebugScale=25,DebugSegments=20,DebugThickness=0.25)
    DriverAxis=(X=1,Y=0,Z=0)
    DriverItem=(Type=Bone,Name="deform_upperarm_l")
    FalloffRegionScaleFactors=(PositiveWidth=0.5,NegativeWidth=0.5,PositiveHeight=0.5,NegativeHeight=0.5)
    FalloffSize=0.45
    OptionalParentItem=(Type=Bone,Name="spine_05")
    RotationOffset=(X=180,Y=40,Z=0)
   )
   > &RigUnit_ModifyTransforms(
    ItemToModify=((Item=(Type=Bone,Name="lat_l"),Transform=(Rotation=(X=0,Y=0,Z=0,W=1),Translation=(X=0,Y=-5,Z=5),Scale3D=(X=1,Y=1,Z=1))))
    WeightMinimum=0
    Weight=@n13.OutputParam
   )
   > n14 = &RigUnit_SphericalPoseReader(
    ActiveRegionScaleFactors=(PositiveWidth=0.1,NegativeWidth=0.1,PositiveHeight=0.1,NegativeHeight=0.1)
    ActiveRegionSize=0.1
    Debug=(bDrawDebug=false,bDraw2D=false,DebugScale=25,DebugSegments=20,DebugThickness=0.25)
    DriverAxis=(X=-1,Y=0,Z=0)
    DriverItem=(Type=Bone,Name="deform_upperarm_r")
    FalloffRegionScaleFactors=(PositiveWidth=0.5,NegativeWidth=0.5,PositiveHeight=0.5,NegativeHeight=0.5)
    FalloffSize=0.45
    OptionalParentItem=(Type=Bone,Name="spine_05")
    RotationOffset=(X=0,Y=40,Z=0)
   )
   > &RigUnit_ModifyTransforms(
    ItemToModify=((Item=(Type=Bone,Name="lat_l"),Transform=(Rotation=(X=0,Y=0,Z=0,W=1),Translation=(X=0,Y=5,Z=-5),Scale3D=(X=1,Y=1,Z=1))))
    WeightMinimum=0
    Weight=@n14.OutputParam
   )
   > n15 = &RigUnit_SphericalPoseReader(
    ActiveRegionScaleFactors=(PositiveWidth=0.1,NegativeWidth=0.1,PositiveHeight=0.1,NegativeHeight=0.1)
    ActiveRegionSize=0.1
    Debug=(bDrawDebug=false,bDraw2D=false,DebugScale=25,DebugSegments=20,DebugThickness=0.25)
    DriverAxis=(X=1,Y=0,Z=0)
    DriverItem=(Type=Bone,Name="deform_upperarm_l")
    FalloffRegionScaleFactors=(PositiveWidth=0.5,NegativeWidth=0.5,PositiveHeight=0.5,NegativeHeight=0.5)
    FalloffSize=0.45
    OptionalParentItem=(Type=Bone,Name="spine_05")
    RotationOffset=(X=200,Y=30,Z=0)
   )
   > &RigUnit_ModifyTransforms(
    ItemToModify=((Item=(Type=Bone,Name="lat_l"),Transform=(Rotation=(X=0,Y=0,Z=0,W=1),Translation=(X=0,Y=-15,Z=15),Scale3D=(X=1,Y=1,Z=1))))
    WeightMinimum=0
    Weight=@n15.OutputParam
   )
   > &RigUnit_ModifyTransforms(
    ItemToModify=((Item=(Type=Bone,Name="trap_l"),Transform=(Rotation=(X=0,Y=0,Z=0,W=1),Translation=(X=0,Y=-4,Z=4),Scale3D=(X=1,Y=1,Z=1))))
    WeightMinimum=0
    Weight=@n15.OutputParam
   )
   > n16 = &RigUnit_SphericalPoseReader(
    ActiveRegionScaleFactors=(PositiveWidth=0.1,NegativeWidth=0.1,PositiveHeight=0.1,NegativeHeight=0.1)
    ActiveRegionSize=0.1
    Debug=(bDrawDebug=false,bDraw2D=false,DebugScale=25,DebugSegments=20,DebugThickness=0.25)
    DriverAxis=(X=-1,Y=0,Z=0)
    DriverItem=(Type=Bone,Name="deform_upperarm_r")
    FalloffRegionScaleFactors=(PositiveWidth=0.5,NegativeWidth=0.5,PositiveHeight=0.5,NegativeHeight=0.5)
    FalloffSize=0.45
    OptionalParentItem=(Type=Bone,Name="spine_05")
    RotationOffset=(X=20,Y=30,Z=0)
   )
   > &RigUnit_ModifyTransforms(
    ItemToModify=((Item=(Type=Bone,Name="lat_r"),Transform=(Rotation=(X=0,Y=0,Z=0,W=1),Translation=(X=0,Y=15,Z=-15),Scale3D=(X=1,Y=1,Z=1))))
    WeightMinimum=0
    Weight=@n16.OutputParam
   )
   > &RigUnit_ModifyTransforms(
    ItemToModify=((Item=(Type=Bone,Name="trap_r"),Transform=(Rotation=(X=0,Y=0,Z=0,W=1),Translation=(X=0,Y=4,Z=-4),Scale3D=(X=1,Y=1,Z=1))))
    WeightMinimum=0
    Weight=@n16.OutputParam
   )
   > n17 = &RigUnit_SphericalPoseReader(
    ActiveRegionScaleFactors=(PositiveWidth=0.1,NegativeWidth=0.1,PositiveHeight=0.1,NegativeHeight=0.1)
    ActiveRegionSize=0.1
    Debug=(bDrawDebug=false,bDraw2D=false,DebugScale=25,DebugSegments=20,DebugThickness=0.25)
    DriverAxis=(X=1,Y=0,Z=0)
    DriverItem=(Type=Bone,Name="deform_upperarm_l")
    FalloffRegionScaleFactors=(PositiveWidth=0.5,NegativeWidth=0.5,PositiveHeight=0.5,NegativeHeight=0.5)
    FalloffSize=0.45
    OptionalParentItem=(Type=Bone,Name="spine_05")
    RotationOffset=(X=180,Y=40,Z=0)
   )
   > &RigUnit_ModifyTransforms(
    ItemToModify=((Item=(Type=Bone,Name="deform_up_arm_twist_01_l"),Transform=(Rotation=(X=0,Y=0,Z=0,W=1),Translation=(X=0,Y=5,Z=0),Scale3D=(X=1,Y=1,Z=1))))
    WeightMinimum=0
    Weight=@n17.OutputParam
   )
   > &RigUnit_ModifyTransforms(
    ItemToModify=((Item=(Type=Bone,Name="deform_up_arm_twist_02_l"),Transform=(Rotation=(X=0,Y=0,Z=0,W=1),Translation=(X=0,Y=5,Z=0),Scale3D=(X=1,Y=1,Z=1))))
    WeightMinimum=0
    Weight=@n17.OutputParam
   )
   > n18 = &RigUnit_SphericalPoseReader(
    ActiveRegionScaleFactors=(PositiveWidth=0.1,NegativeWidth=0.1,PositiveHeight=0.1,NegativeHeight=0.1)
    ActiveRegionSize=0.1
    Debug=(bDrawDebug=false,bDraw2D=false,DebugScale=25,DebugSegments=20,DebugThickness=0.25)
    DriverAxis=(X=-1,Y=0,Z=0)
    DriverItem=(Type=Bone,Name="deform_upperarm_r")
    FalloffRegionScaleFactors=(PositiveWidth=0.5,NegativeWidth=0.5,PositiveHeight=0.5,NegativeHeight=0.5)
    FalloffSize=0.45
    OptionalParentItem=(Type=Bone,Name="spine_05")
    RotationOffset=(X=0,Y=40,Z=0)
   )
   > &RigUnit_ModifyTransforms(
    ItemToModify=((Item=(Type=Bone,Name="deform_up_arm_twist_01_r"),Transform=(Rotation=(X=0,Y=0,Z=0,W=1),Translation=(X=0,Y=-5,Z=0),Scale3D=(X=1,Y=1,Z=1))))
    WeightMinimum=0
    Weight=@n18.OutputParam
   )
   > &RigUnit_ModifyTransforms(
    ItemToModify=((Item=(Type=Bone,Name="deform_up_arm_twist_02_r"),Transform=(Rotation=(X=0,Y=0,Z=0,W=1),Translation=(X=0,Y=-5,Z=0),Scale3D=(X=1,Y=1,Z=1))))
    WeightMinimum=0
    Weight=@n18.OutputParam
   )
   > n19 = &RigUnit_SphericalPoseReader(
    ActiveRegionScaleFactors=(PositiveWidth=0.1,NegativeWidth=0.1,PositiveHeight=0.1,NegativeHeight=0.1)
    ActiveRegionSize=0.1
    Debug=(bDrawDebug=false,bDraw2D=false,DebugScale=25,DebugSegments=20,DebugThickness=0.25)
    DriverAxis=(X=1,Y=0,Z=0)
    DriverItem=(Type=Bone,Name="deform_upperarm_l")
    FalloffRegionScaleFactors=(PositiveWidth=0.4,NegativeWidth=0.4,PositiveHeight=0.4,NegativeHeight=0.4)
    FalloffSize=0.45
    OptionalParentItem=(Type=Bone,Name="spine_05")
    RotationOffset=(X=50,Y=0,Z=10)
   )
   > &RigUnit_ModifyTransforms(
    ItemToModify=((Item=(Type=Bone,Name="deform_up_arm_twist_01_l"),Transform=(Rotation=(X=0,Y=0,Z=0,W=1),Translation=(X=-1,Y=0,Z=-1),Scale3D=(X=1,Y=1,Z=1))))
    WeightMinimum=0
    Weight=@n19.OutputParam
   )
   > &RigUnit_ModifyTransforms(
    ItemToModify=((Item=(Type=Bone,Name="deform_clavicle_aim_l"),Transform=(Rotation=(X=0,Y=0,Z=0,W=1),Translation=(X=0,Y=0,Z=10),Scale3D=(X=1,Y=1,Z=1))))
    WeightMinimum=0
    Weight=@n19.OutputParam
   )
   > n20 = &RigUnit_SphericalPoseReader(
    ActiveRegionScaleFactors=(PositiveWidth=0.1,NegativeWidth=0.1,PositiveHeight=0.1,NegativeHeight=0.1)
    ActiveRegionSize=0.1
    Debug=(bDrawDebug=false,bDraw2D=false,DebugScale=25,DebugSegments=20,DebugThickness=0.25)
    DriverAxis=(X=-1,Y=0,Z=0)
    DriverItem=(Type=Bone,Name="deform_upperarm_r")
    FalloffRegionScaleFactors=(PositiveWidth=0.4,NegativeWidth=0.4,PositiveHeight=0.4,NegativeHeight=0.4)
    FalloffSize=0.45
    OptionalParentItem=(Type=Bone,Name="spine_05")
    RotationOffset=(X=-90,Y=0,Z=30)
   )
   > &RigUnit_ModifyTransforms(
    ItemToModify=((Item=(Type=Bone,Name="deform_up_arm_twist_01_r"),Transform=(Rotation=(X=0,Y=0,Z=0,W=1),Translation=(X=1,Y=0,Z=1),Scale3D=(X=1,Y=1.1,Z=1))))
    WeightMinimum=0
    Weight=@n20.OutputParam
   )
   > &RigUnit_ModifyTransforms(
    ItemToModify=((Item=(Type=Bone,Name="deform_clavicle_aim_r"),Transform=(Rotation=(X=0,Y=0,Z=0,W=1),Translation=(X=0,Y=0,Z=-10),Scale3D=(X=1,Y=1,Z=1))))
    WeightMinimum=0
    Weight=@n20.OutputParam
   )
   RigVMFunction_MathQuaternionSwingTwist = &RigVMFunction_MathQuaternionSwingTwist(
    TwistAxis=(X=1,Y=0,Z=0)
    Input=&RigUnit_GetTransform(Item=(Type=Bone,Name="deform_upperarm_l"), Space=LocalSpace, bInitial=False).Transform.Rotation
   )
   Inverse = /Script/RigVM.RigVMFunction_MathQuaternionInverse(Value=@RigVMFunction_MathQuaternionSwingTwist.Twist) {is=struct RigVMUnitNode, template="Inverse::Execute(in Value,out Result)"}
   > Set Transform = &RigUnit_SetRotation(Item=(Type=Bone,Name="deform_up_arm_twist_01_l"), Weight=0.7, bInitial=false, Value=@Inverse.Result)
   > Set Transform_1 = &RigUnit_SetRotation(Item=(Type=Bone,Name="deform_up_arm_twist_02_l"), Weight=0.5, bInitial=false, Value=@Inverse.Result)
   > Set Transform_1_1 = &RigUnit_SetRotation(Item=(Type=Bone,Name="deform_up_arm_twist_03_l"), Weight=0.3, bInitial=false, Value=@Inverse.Result)
   n21 = &RigUnit_GetTransform(Item=(Type=Bone,Name="upperarm_retarg_r"), Space=LocalSpace, bInitial=False)
   > Set Transform_2 = &RigUnit_SetRotation(
    Item=(Type=Bone,Name="deform_up_arm_twist_01_r")
    Weight=0.7
    bInitial=false
    Value=&RigVMFunction_MathQuaternionInverse(Value=&RigVMFunction_MathQuaternionSwingTwist(TwistAxis=(X=-1,Y=0,Z=0), Input=@n21.Transform.Rotation).Twist) {template="Inverse::Execute(in Value,out Result)"}.Result
   )
   > Set Transform_1_2 = &RigUnit_SetRotation(Item=(Type=Bone,Name="deform_up_arm_twist_02_r"), Value=(X=0,Y=0,Z=0,W=1), Weight=0.5, bInitial=false)
   > Set Transform_1_1_2 = &RigUnit_SetRotation(Item=(Type=Bone,Name="deform_up_arm_twist_03_r"), Value=(X=0,Y=0,Z=0,W=1), Weight=0.3, bInitial=False)
   > &RigUnit_RotationConstraintLocalSpaceOffset(
    AdvancedSettings=(InterpolationType=Average,RotationOrderForFilter=XZY)
    Child=(Type=Bone,Name="deform_handSplit_01_l")
    Filter=(bX=True,bY=True,bZ=True)
    Parents=((Item=(Type=Bone,Name="deform_hand_l"),Weight=1))
    Weight=0.5
   )
   > &RigUnit_RotationConstraintLocalSpaceOffset(
    AdvancedSettings=(InterpolationType=Average,RotationOrderForFilter=XZY)
    Child=(Type=Bone,Name="deform_handSplit_01_r")
    Filter=(bX=True,bY=True,bZ=True)
    Parents=((Item=(Type=Bone,Name="deform_hand_r"),Weight=1))
    Weight=0.5
   )
   > Set Transform_1_1_3 = &RigUnit_SetRotation(
    Item=(Type=Bone,Name="deform_lo_arm_twistParent_03_l")
    Weight=1
    bInitial=false
    Value=&RigVMFunction_MathQuaternionSwingTwist(
     TwistAxis=(X=1,Y=0,Z=0)
     Input=&RigUnit_GetTransform(Item=(Type=Bone,Name="deform_handSplit_01_l"), Space=LocalSpace, bInitial=False).Transform.Rotation
    ).Twist
   )
   > Set Transform_1_1_3_1 = &RigUnit_SetRotation(
    Item=(Type=Bone,Name="deform_lo_arm_twist_03_l")
    Weight=1
    bInitial=false
    Value=&RigUnit_GetTransform(Item=(Type=Bone,Name="deform_lo_arm_twistParent_03_l"), Space=LocalSpace, bInitial=False).Transform.Rotation
   )
   > &RigUnit_ModifyTransforms(
    ItemToModify=((Item=(Type=Bone,Name="deform_lo_arm_twist_03_l"),Transform=(Rotation=(X=-1,Y=0,Z=0,W=0),Translation=(X=0,Y=0,Z=0),Scale3D=(X=1,Y=1,Z=1))))
    Weight=1
    WeightMinimum=0
   )
   > Set Transform_1_1_3_2 = &RigUnit_SetRotation(
    Item=(Type=Bone,Name="deform_lo_arm_twistParent_03_r")
    Weight=1
    bInitial=false
    Value=&RigVMFunction_MathQuaternionSwingTwist(
     TwistAxis=(X=1,Y=0,Z=0)
     Input=&RigUnit_GetTransform(Item=(Type=Bone,Name="deform_handSplit_01_r"), Space=LocalSpace, bInitial=False).Transform.Rotation
    ).Twist
   )
   > Set Transform_1_1_3_1_1 = &RigUnit_SetRotation(
    Item=(Type=Bone,Name="deform_lo_arm_twist_03_r")
    Weight=1
    bInitial=false
    Value=&RigUnit_GetTransform(Item=(Type=Bone,Name="deform_lo_arm_twistParent_03_r"), Space=LocalSpace, bInitial=False).Transform.Rotation
   )
   > &RigUnit_ModifyTransforms(
    ItemToModify=((Item=(Type=Bone,Name="deform_lo_arm_twist_03_r"),Transform=(Rotation=(X=-1,Y=0,Z=0,W=0),Translation=(X=0,Y=0,Z=0),Scale3D=(X=1,Y=1,Z=1))))
    Weight=1
    WeightMinimum=0
   )
  ]}
]
hierarchy [
 root {loc=<0 0 0>, rot=<0 0 0 1>}
 pelvis {parent=root, loc=<0 -0.7071 93.1191>, rot=<0.0127 0.707 -0.0127 -0.707>}
 spine_01 {parent=pelvis, loc=<6.7494 -0.6388 0>, rot=<0 0 0.034 -0.9994>}
 spine_02 {parent=spine_01, loc=<7.2518 -0.035 0>, rot=<0 0 0.0238 -0.9997>}
 spine_03 {parent=spine_02, loc=<6.0104 0 0>, rot=<0 0 0.0149 -0.9999>}
 spine_04 {parent=spine_03, loc=<8.1024 0 0>, rot=<0 0 -0.0132 -0.9999>}
 spine_05 {parent=spine_04, loc=<10.0146 0 0>, rot=<0 0 -0.0015 -1>}
 attach_cape {parent=spine_05, loc=<9.017 -11.714 0>, rot=<0.0616 -0.7044 -0.0616 -0.7044>}
 neck_01 {parent=spine_05, loc=<11.3138 0.1529 0>, rot=<0 0 -0.2003 -0.9797>}
 neck_02 {parent=neck_01, loc=<4.5627 0 0>, rot=<0 0 0.0504 -0.9987>}
 head {parent=neck_02, loc=<5.3914 0 0>, rot=<0 0 0.115 -0.9934>}
 attach_backpack {parent=spine_05, loc=<-0.123 -13.473 0>, rot=<-0.0093 -0.707 0.0093 -0.707>}
 clavicle_r {parent=spine_05, loc=<4.3575 -0.1466 2.3223>, rot=<0.7289 0.0465 -0.6829 0.009>}
 upperarm_r {parent=clavicle_r, loc=<-15.5111 0 0>, rot=<0.0226 -0.3234 -0.0088 -0.9459>}
 lowerarm_r {parent=upperarm_r, loc=<-26.6514 0 0>, rot=<0.0391 -0.0099 -0.2453 -0.9686>}
 hand_r {parent=lowerarm_r, loc=<-22.3491 0 0>, rot=<-0.7081 -0.0507 0.0975 -0.6975>}
 weapon_r {parent=hand_r, loc=<0.1974 3.4111 -0.3811>, rot=<0 0 0 1>}
 thumb_01_r {parent=hand_r, loc=<-1.3118 0.3191 2.459>, rot=<0.6546 -0.4042 0.0664 -0.6354>}
 thumb_02_r {parent=thumb_01_r, loc=<-5.9336 0.0448 -0.0401>, rot=<-0.0037 0.0213 0.298 -0.9543>}
 thumb_03_r {parent=thumb_02_r, loc=<-2.2223 -0.3682 0.0907>, rot=<0.0118 -0.013 0.0066 -0.9998>}
 index_metacarpal_r {parent=hand_r, loc=<-1.6535 -0.0183 2.96>, rot=<0.0158 -0.0124 -0.0173 -0.9996>}
 index_01_r {parent=index_metacarpal_r, loc=<-8.3008 -0.0357 0.0524>, rot=<-0.0044 0.0485 0.0924 -0.9945>}
 index_02_r {parent=index_01_r, loc=<-3.3587 0.3004 -0.0452>, rot=<-0.002 0.0027 0.3225 -0.9466>}
 index_03_r {parent=index_02_r, loc=<-2.5578 0.123 0.0252>, rot=<-0.002 0.0067 0.2558 -0.9667>}
 middle_metacarpal_r {parent=hand_r, loc=<-2.1856 0.1725 0.5984>, rot=<0.006 0.0168 -0.0329 -0.9993>}
 middle_01_r {parent=middle_metacarpal_r, loc=<-7.3275 0.2447 -0.0129>, rot=<-0.0208 0.0042 0.0858 -0.9961>}
 middle_02_r {parent=middle_01_r, loc=<-4.0123 -0.1401 -0.1175>, rot=<-0.0218 0.0258 0.3512 -0.9357>}
 middle_03_r {parent=middle_02_r, loc=<-2.6077 -0.0534 -0.1373>, rot=<0.0082 0.0188 0.1708 -0.9851>}
 ring_metacarpal_r {parent=hand_r, loc=<-2.2074 0.3625 -1.3638>, rot=<-0.0284 0.0722 -0.0458 -0.9959>}
 ring_01_r {parent=ring_metacarpal_r, loc=<-7.0295 0.6197 0.1446>, rot=<-0.0489 -0.0252 0.1892 -0.9804>}
 ring_02_r {parent=ring_01_r, loc=<-3.6174 -0.4647 0.1027>, rot=<0 0 0.2806 -0.9598>}
 ring_03_r {parent=ring_02_r, loc=<-2.4152 -0.114 -0.0199>, rot=<0 0 0.1189 -0.9929>}
 pinky_metacarpal_r {parent=hand_r, loc=<-1.7616 0.9753 -3.3814>, rot=<-0.0397 0.0814 -0.069 -0.9935>}
 pinky_01_r {parent=pinky_metacarpal_r, loc=<-7.1709 0.4831 0.1094>, rot=<-0.0654 -0.0035 0.2242 -0.9723>}
 pinky_02_r {parent=pinky_01_r, loc=<-2.8274 -0.0893 0.1043>, rot=<0.0037 0.0011 0.2854 -0.9584>}
 pinky_03_r {parent=pinky_02_r, loc=<-2.2046 -0.0307 0.0078>, rot=<-0.0002 0.0098 0.1445 -0.9895>}
 hand_attach_r {parent=hand_r, loc=<-8.5001 2.9999 0>, rot=<0 0 -0.7071 -0.7071>}
 lowerarm_twist_02_r {parent=lowerarm_r, loc=<-9.6199 0 0>, rot=<0 0 0 -1>}
 lowerarm_twist_01_r {parent=lowerarm_r, loc=<-17.8549 0 0>, rot=<0 0 0 -1>}
 upperarm_twist_01_r {parent=upperarm_r, loc=<-0.0041 0 0>, rot=<0 0 0 -1>}
 upperarm_twist_02_r {parent=upperarm_r, loc=<-13.7941 0 0>, rot=<0 0 0 -1>}
 trap_l {parent=spine_05, loc=<14.1419 1.6568 -8.9026>, rot=<0 0 -0.0401 -0.9992>}
 chest_l {parent=spine_05, loc=<-1.9976 -0.7806 -9.8286>, rot=<0 0 -0.0401 -0.9992>}
 chest_r {parent=spine_05, loc=<-1.9979 -0.7807 9.8286>, rot=<0 0 -0.0401 -0.9992>}
 clavicle_l {parent=spine_05, loc=<4.3575 -0.1466 -2.3223>, rot=<0.0465 -0.7289 -0.009 -0.6829>}
 upperarm_l {parent=clavicle_l, loc=<15.5111 0 0>, rot=<0.0226 -0.3234 -0.0088 -0.9459>}
 lowerarm_l {parent=upperarm_l, loc=<26.6514 0 0>, rot=<0.0391 -0.0099 -0.2453 -0.9686>}
 hand_l {parent=lowerarm_l, loc=<22.3491 0 0>, rot=<-0.7081 -0.0507 0.0975 -0.6975>}
 thumb_01_l {parent=hand_l, loc=<1.3118 -0.3191 -2.459>, rot=<0.6546 -0.4042 0.0664 -0.6354>}
 thumb_02_l {parent=thumb_01_l, loc=<5.9336 -0.0448 0.0401>, rot=<-0.0037 0.0213 0.298 -0.9543>}
 thumb_03_l {parent=thumb_02_l, loc=<2.2223 0.3682 -0.0907>, rot=<0.0118 -0.013 0.0066 -0.9998>}
 index_metacarpal_l {parent=hand_l, loc=<1.6535 0.0183 -2.96>, rot=<0.0158 -0.0124 -0.0173 -0.9996>}
 index_01_l {parent=index_metacarpal_l, loc=<8.3008 0.0357 -0.0524>, rot=<-0.0044 0.0485 0.0924 -0.9945>}
 index_02_l {parent=index_01_l, loc=<3.3587 -0.3004 0.0452>, rot=<-0.002 0.0027 0.3225 -0.9466>}
 index_03_l {parent=index_02_l, loc=<2.5578 -0.123 -0.0252>, rot=<-0.002 0.0067 0.2558 -0.9667>}
 middle_metacarpal_l {parent=hand_l, loc=<2.1856 -0.1725 -0.5984>, rot=<0.006 0.0168 -0.0329 -0.9993>}
 middle_01_l {parent=middle_metacarpal_l, loc=<7.3275 -0.2447 0.0129>, rot=<-0.0208 0.0042 0.0858 -0.9961>}
 middle_02_l {parent=middle_01_l, loc=<4.0123 0.1401 0.1175>, rot=<-0.0218 0.0258 0.3512 -0.9357>}
 middle_03_l {parent=middle_02_l, loc=<2.6077 0.0534 0.1373>, rot=<0.0082 0.0188 0.1708 -0.9851>}
 ring_metacarpal_l {parent=hand_l, loc=<2.2074 -0.3625 1.3638>, rot=<-0.0284 0.0722 -0.0458 -0.9959>}
 ring_01_l {parent=ring_metacarpal_l, loc=<7.0295 -0.6197 -0.1446>, rot=<-0.0489 -0.0252 0.1892 -0.9804>}
 ring_02_l {parent=ring_01_l, loc=<3.6174 0.4647 -0.1027>, rot=<0 0 0.2806 -0.9598>}
 ring_03_l {parent=ring_02_l, loc=<2.4152 0.114 0.0199>, rot=<0 0 0.1189 -0.9929>}
 pinky_metacarpal_l {parent=hand_l, loc=<1.7616 -0.9753 3.3814>, rot=<-0.0397 0.0814 -0.069 -0.9935>}
 pinky_01_l {parent=pinky_metacarpal_l, loc=<7.1709 -0.4831 -0.1094>, rot=<-0.0654 -0.0035 0.2242 -0.9723>}
 pinky_02_l {parent=pinky_01_l, loc=<2.8274 0.0893 -0.1043>, rot=<0.0037 0.0011 0.2854 -0.9584>}
 pinky_03_l {parent=pinky_02_l, loc=<2.2046 0.0307 -0.0078>, rot=<-0.0002 0.0098 0.1445 -0.9895>}
 weapon_l {parent=hand_l, loc=<-0.1974 -3.4111 0.3811>, rot=<0 0 0 1>}
 hand_attach_l {parent=hand_l, loc=<8.5 -3 0>, rot=<0.7071 0.7071 0 0>}
 lowerarm_twist_02_l {parent=lowerarm_l, loc=<9.6199 0 0>, rot=<0 0 0 -1>}
 lowerarm_twist_01_l {parent=lowerarm_l, loc=<17.8549 0 0>, rot=<0 0 0 -1>}
 lo_arm_twist_04_l {parent=lowerarm_l, loc=<22.3491 0 0>, rot=<0 0 0 1>}
 upperarm_twist_01_l {parent=upperarm_l, loc=<0.0041 0 0>, rot=<0 0 0 1>}
 upperarm_twist_02_l {parent=upperarm_l, loc=<13.7941 0 0>, rot=<0 0 0 1>}
 clavicle_retarget_l {parent=spine_05, loc=<4.3575 -0.1466 -2.3223>, rot=<0.0465 -0.7289 -0.009 -0.6829>}
 upperarm_retarg_l {parent=clavicle_retarget_l, loc=<15.2278 0.4299 4.2826>, rot=<0.0474 -0.3501 0.0186 -0.9353>}
 lowerarm_retarg_l {parent=upperarm_retarg_l, loc=<27.4087 0 0>, rot=<-0.025 -0.0251 -0.2825 -0.9586>}
 hand_retarg_l {parent=lowerarm_retarg_l, loc=<25.1351 0 0>, rot=<-0.6625 -0.0177 0.1299 -0.7375>}
 pinky_metacarpal_retarg_l {parent=hand_retarg_l, loc=<2.026 -0.7994 3.2779>, rot=<-0.0381 0.0493 -0.1029 -0.9927>}
 pinky_retarg_01_l {parent=pinky_metacarpal_retarg_l, loc=<7.1709 -0.4831 -0.1094>, rot=<-0.0654 -0.0035 0.2242 -0.9723>}
 pinky_retarg_02_l {parent=pinky_retarg_01_l, loc=<2.8274 0.0893 -0.1043>, rot=<0.0037 0.0011 0.2854 -0.9584>}
 pinky_retarg_03_l {parent=pinky_retarg_02_l, loc=<2.2046 0.0307 -0.0078>, rot=<-0.0002 0.0098 0.1445 -0.9895>}
 ring_metacarpal_retarg_l {parent=hand_retarg_l, loc=<2.3054 -0.1877 1.2303>, rot=<-0.0258 0.0406 -0.0802 -0.9956>}
 ring_retarg_01_l {parent=ring_metacarpal_retarg_l, loc=<7.0295 -0.6197 -0.1446>, rot=<-0.0489 -0.0252 0.1892 -0.9804>}
 ring_retarg_02_l {parent=ring_retarg_01_l, loc=<3.6174 0.4647 -0.1027>, rot=<0 0 0.2806 -0.9598>}
 ring_retarg_03_l {parent=ring_retarg_02_l, loc=<2.4152 0.114 0.0199>, rot=<0 0 0.1189 -0.9929>}
 middle_metacarpal_retarg_l {parent=hand_retarg_l, loc=<2.1524 -0.0293 -0.7289>, rot=<0.0109 -0.0136 -0.068 -0.9975>}
 middle_retarg_01_l {parent=middle_metacarpal_retarg_l, loc=<7.3275 -0.2447 0.0129>, rot=<-0.0208 0.0042 0.0858 -0.9961>}
 middle_retarg_02_l {parent=middle_retarg_01_l, loc=<4.0123 0.1401 0.1175>, rot=<-0.0218 0.0258 0.3512 -0.9357>}
 middle_retarg_03_l {parent=middle_retarg_02_l, loc=<2.6077 0.0534 0.1373>, rot=<0.0082 0.0188 0.1708 -0.9851>}
 index_metacarpal_retarg_l {parent=hand_retarg_l, loc=<1.467 0.0884 -3.0555>, rot=<0.0221 -0.0423 -0.0527 -0.9975>}
 index_retarg_01_l {parent=index_metacarpal_retarg_l, loc=<8.3008 0.0357 -0.0524>, rot=<-0.0044 0.0485 0.0924 -0.9945>}
 index_retarg_02_l {parent=index_retarg_01_l, loc=<3.3587 -0.3004 0.0452>, rot=<-0.002 0.0027 0.3225 -0.9466>}
 index_retarg_03_l {parent=index_retarg_02_l, loc=<2.5578 -0.123 -0.0252>, rot=<-0.002 0.0067 0.2558 -0.9667>}
 thumb_retarg_01_l {parent=hand_retarg_l, loc=<1.1806 -0.2643 -2.5309>, rot=<0.6741 -0.3997 0.0268 -0.6205>}
 thumb_retarg_02_l {parent=thumb_retarg_01_l, loc=<5.9336 -0.0448 0.0401>, rot=<-0.0037 0.0213 0.298 -0.9543>}
 thumb_retarg_03_l {parent=thumb_retarg_02_l, loc=<2.2223 0.3682 -0.0907>, rot=<0.0118 -0.013 0.0066 -0.9998>}
 lo_arm_twist_01_l {parent=lowerarm_retarg_l, loc=<8.4062 0 0>, rot=<0 0 0 1>}
 lo_arm_twist_02_l {parent=lowerarm_retarg_l, loc=<13.9825 0 0>, rot=<0 0 0 1>}
 lowerarm_ot_l {parent=lowerarm_retarg_l, loc=<0 0 0>, rot=<0.0687 0.0306 0.0067 -0.9971>}
 lowerarm_in_l {parent=lowerarm_retarg_l, loc=<0 0 0>, rot=<0.0687 0.0306 0.0067 -0.9971>}
 handSplit_01_retarg_l {parent=lowerarm_retarg_l, loc=<25.1351 0 0>, rot=<-0.6625 -0.0177 0.1299 -0.7375>}
 handSplit_02_retarg_l {parent=handSplit_01_retarg_l, loc=<0 0 0>, rot=<0 0 0 1>}
 handStop_retarg_l {parent=lowerarm_retarg_l, loc=<25.1351 0 0>, rot=<-0.6625 -0.0177 0.1299 -0.7375>}
 lo_arm_twistParent_03_l {parent=lowerarm_retarg_l, loc=<19.5588 0 0>, rot=<0 0 0 1>}
 lo_arm_twist_03_l {parent=lo_arm_twistParent_03_l, loc=<0 0 0>, rot=<0 0 0 1>}
 up_arm_twist_02_l {parent=upperarm_retarg_l, loc=<9.6764 0 0>, rot=<-0.0146 0.0295 -0.0332 -0.9989>}
 up_arm_twist_04_l {parent=upperarm_retarg_l, loc=<22.9496 0 0>, rot=<0 0 0 1>}
 bicep_04_l {parent=up_arm_twist_04_l, loc=<0 0 0>, rot=<-0.0146 0.0295 -0.0332 -0.9989>}
 tricep_04_l {parent=up_arm_twist_04_l, loc=<0 0 0>, rot=<-0.0146 0.0295 -0.0332 -0.9989>}
 up_arm_twist_01_l {parent=upperarm_retarg_l, loc=<0 0 0>, rot=<-0.0146 0.0295 -0.0332 -0.9989>}
 up_arm_twist_03_l {parent=upperarm_retarg_l, loc=<16.313 0 0>, rot=<-0.0146 0.0295 -0.0332 -0.9989>}
 bicep_03_l {parent=up_arm_twist_03_l, loc=<0 0 0>, rot=<0 0 0 -1>}
 tricep_03_l {parent=up_arm_twist_03_l, loc=<0 0 0>, rot=<0 0 0 -1>}
 clavicle_aim_retarget_l {parent=clavicle_retarget_l, loc=<0 0 0>, rot=<0 0 0 1>}
 trap_r {parent=spine_05, loc=<14.1418 1.6568 8.9025>, rot=<0 0 0.9992 -0.0401>}
 clavicle_retarget_r {parent=spine_05, loc=<4.3578 -0.1466 2.3223>, rot=<0.7289 0.0465 -0.6829 0.009>}
 upperarm_retarg_r {parent=clavicle_retarget_r, loc=<-15.2278 -0.43 -4.2822>, rot=<0.0474 -0.3501 0.0186 -0.9353>}
 lowerarm_retarg_r {parent=upperarm_retarg_r, loc=<-27.4084 0 -0.0002>, rot=<-0.025 -0.0251 -0.2825 -0.9586>}
 hand_retarg_r {parent=lowerarm_retarg_r, loc=<-25.1349 -0.0001 -0.0002>, rot=<-0.6625 -0.0177 0.1299 -0.7375>}
 pinky_metacarpal_retarg_r {parent=hand_retarg_r, loc=<-2.0265 0.7996 -3.2779>, rot=<-0.0381 0.0493 -0.1029 -0.9927>}
 pinky_retarg_01_r {parent=pinky_metacarpal_retarg_r, loc=<-7.171 0.483 0.1094>, rot=<-0.0654 -0.0035 0.2242 -0.9723>}
 pinky_retarg_02_r {parent=pinky_retarg_01_r, loc=<-2.8274 -0.0892 0.1043>, rot=<0.0037 0.0011 0.2854 -0.9584>}
 pinky_retarg_03_r {parent=pinky_retarg_02_r, loc=<-2.2046 -0.0307 0.0078>, rot=<-0.0002 0.0098 0.1445 -0.9895>}
 ring_metacarpal_retarg_r {parent=hand_retarg_r, loc=<-2.306 0.188 -1.2304>, rot=<-0.0258 0.0406 -0.0802 -0.9956>}
 ring_retarg_01_r {parent=ring_metacarpal_retarg_r, loc=<-7.0293 0.6196 0.1446>, rot=<-0.0489 -0.0252 0.1892 -0.9804>}
 ring_retarg_02_r {parent=ring_retarg_01_r, loc=<-3.6173 -0.4646 0.1027>, rot=<0 0 0.2806 -0.9598>}
 ring_retarg_03_r {parent=ring_retarg_02_r, loc=<-2.4153 -0.1141 -0.0199>, rot=<0 0 0.1189 -0.9929>}
 middle_metacarpal_retarg_r {parent=hand_retarg_r, loc=<-2.1526 0.0294 0.7288>, rot=<0.0109 -0.0136 -0.068 -0.9975>}
 middle_retarg_01_r {parent=middle_metacarpal_retarg_r, loc=<-7.3277 0.2449 -0.013>, rot=<-0.0208 0.0042 0.0858 -0.9961>}
 middle_retarg_02_r {parent=middle_retarg_01_r, loc=<-4.0123 -0.1401 -0.1175>, rot=<-0.0218 0.0258 0.3512 -0.9357>}
 middle_retarg_03_r {parent=middle_retarg_02_r, loc=<-2.6077 -0.0534 -0.1373>, rot=<0.0082 0.0188 0.1708 -0.9851>}
 index_metacarpal_retarg_r {parent=hand_retarg_r, loc=<-1.4673 -0.0882 3.0555>, rot=<0.0221 -0.0423 -0.0527 -0.9975>}
 index_retarg_01_r {parent=index_metacarpal_retarg_r, loc=<-8.3009 -0.0357 0.0524>, rot=<-0.0044 0.0485 0.0924 -0.9945>}
 index_retarg_02_r {parent=index_retarg_01_r, loc=<-3.3588 0.3004 -0.0452>, rot=<-0.002 0.0027 0.3225 -0.9466>}
 index_retarg_03_r {parent=index_retarg_02_r, loc=<-2.5578 0.123 0.0252>, rot=<-0.002 0.0067 0.2558 -0.9667>}
 thumb_retarg_01_r {parent=hand_retarg_r, loc=<-1.1809 0.2645 2.5309>, rot=<0.6741 -0.3997 0.0268 -0.6205>}
 thumb_retarg_02_r {parent=thumb_retarg_01_r, loc=<-5.9337 0.045 -0.0402>, rot=<-0.0037 0.0213 0.298 -0.9543>}
 thumb_retarg_03_r {parent=thumb_retarg_02_r, loc=<-2.2222 -0.3682 0.0907>, rot=<0.0118 -0.013 0.0066 -0.9998>}
 lo_arm_twist_01_r {parent=lowerarm_retarg_r, loc=<-8.4062 0 -0.0001>, rot=<0 0 0 1>}
 lo_arm_twist_02_r {parent=lowerarm_retarg_r, loc=<-13.9829 0.0002 0.0003>, rot=<0 0 0 1>}
 lowerarm_ot_r {parent=lowerarm_retarg_r, loc=<0 0 0>, rot=<0.0687 0.0306 0.0067 -0.9971>}
 lowerarm_in_r {parent=lowerarm_retarg_r, loc=<0 0 0>, rot=<0.0687 0.0306 0.0067 -0.9971>}
 handSplit_01_retarg_r {parent=lowerarm_retarg_r, loc=<-25.1349 -0.0001 -0.0002>, rot=<-0.6625 -0.0177 0.1299 -0.7375>}
 handSplit_02_retarg_r {parent=handSplit_01_retarg_r, loc=<0 0 0>, rot=<0 0 0 1>}
 handStop_retarg_r {parent=lowerarm_retarg_r, loc=<-25.1349 -0.0001 -0.0002>, rot=<-0.6625 -0.0177 0.1299 -0.7375>}
 lo_arm_twistParent_03_r {parent=lowerarm_retarg_r, loc=<-19.5589 0 0.0001>, rot=<0 0 0 1>}
 lo_arm_twist_03_r {parent=lo_arm_twistParent_03_r, loc=<0 0 0>, rot=<0 0 0 1>}
 up_arm_twist_02_r {parent=upperarm_retarg_r, loc=<-9.6764 0 -0.0001>, rot=<-0.0146 0.0295 -0.0332 -0.9989>}
 up_arm_twist_03_r {parent=upperarm_retarg_r, loc=<-16.3126 0.0001 -0.0004>, rot=<-0.0146 0.0295 -0.0332 -0.9989>}
 bicep_03_r {parent=up_arm_twist_03_r, loc=<0 0 0>, rot=<0 0 0 -1>}
 tricep_03_r {parent=up_arm_twist_03_r, loc=<0 0 0>, rot=<0 0 0 -1>}
 up_arm_twist_04_r {parent=upperarm_retarg_r, loc=<-22.9496 0 -0.0001>, rot=<0 0 0 1>}
 bicep_04_r {parent=up_arm_twist_04_r, loc=<0 0 0>, rot=<-0.0146 0.0295 -0.0332 -0.9989>}
 tricep_04_r {parent=up_arm_twist_04_r, loc=<0 0 0>, rot=<-0.0146 0.0295 -0.0332 -0.9989>}
 up_arm_twist_01_r {parent=upperarm_retarg_r, loc=<0 0 0>, rot=<-0.0146 0.0295 -0.0332 -0.9989>}
 clavicle_aim_retarget_r {parent=clavicle_retarget_r, loc=<0 0 0>, rot=<0 0 0 1>}
 deform_clavicle_l {parent=spine_05, loc=<4.3575 -0.1466 -2.3223>, rot=<0.0465 -0.7289 -0.009 -0.6829>}
 deform_upperarm_l {parent=deform_clavicle_l, loc=<15.2278 0.4299 4.2826>, rot=<0.0474 -0.3501 0.0186 -0.9353>}
 deform_lowerarm_l {parent=deform_upperarm_l, loc=<27.4087 0 0>, rot=<-0.025 -0.0251 -0.2825 -0.9586>}
 deform_hand_l {parent=deform_lowerarm_l, loc=<25.1351 0 0>, rot=<-0.6625 -0.0177 0.1299 -0.7375>}
 deform_pinky_metacarpal_l {parent=deform_hand_l, loc=<2.026 -0.7994 3.2779>, rot=<-0.0381 0.0493 -0.1029 -0.9927>}
 deform_pinky_01_l {parent=deform_pinky_metacarpal_l, loc=<7.1709 -0.4831 -0.1094>, rot=<-0.0654 -0.0035 0.2242 -0.9723>}
 deform_pinky_02_l {parent=deform_pinky_01_l, loc=<2.8274 0.0893 -0.1043>, rot=<0.0037 0.0011 0.2854 -0.9584>}
 deform_pinky_03_l {parent=deform_pinky_02_l, loc=<2.2046 0.0307 -0.0078>, rot=<-0.0002 0.0098 0.1445 -0.9895>}
 deform_ring_metacarpal_l {parent=deform_hand_l, loc=<2.3054 -0.1877 1.2303>, rot=<-0.0258 0.0406 -0.0802 -0.9956>}
 deform_ring_01_l {parent=deform_ring_metacarpal_l, loc=<7.0295 -0.6197 -0.1446>, rot=<-0.0489 -0.0252 0.1892 -0.9804>}
 deform_ring_02_l {parent=deform_ring_01_l, loc=<3.6174 0.4647 -0.1027>, rot=<0 0 0.2806 -0.9598>}
 deform_ring_03_l {parent=deform_ring_02_l, loc=<2.4152 0.114 0.0199>, rot=<0 0 0.1189 -0.9929>}
 deform_middle_metacarpal_l {parent=deform_hand_l, loc=<2.1524 -0.0293 -0.7289>, rot=<0.0109 -0.0136 -0.068 -0.9975>}
 deform_middle_01_l {parent=deform_middle_metacarpal_l, loc=<7.3275 -0.2447 0.0129>, rot=<-0.0208 0.0042 0.0858 -0.9961>}
 deform_middle_02_l {parent=deform_middle_01_l, loc=<4.0123 0.1401 0.1175>, rot=<-0.0218 0.0258 0.3512 -0.9357>}
 deform_middle_03_l {parent=deform_middle_02_l, loc=<2.6077 0.0534 0.1373>, rot=<0.0082 0.0188 0.1708 -0.9851>}
 deform_index_metacarpal_l {parent=deform_hand_l, loc=<1.467 0.0884 -3.0555>, rot=<0.0221 -0.0423 -0.0527 -0.9975>}
 deform_index_01_l {parent=deform_index_metacarpal_l, loc=<8.3008 0.0357 -0.0524>, rot=<-0.0044 0.0485 0.0924 -0.9945>}
 deform_index_02_l {parent=deform_index_01_l, loc=<3.3587 -0.3004 0.0452>, rot=<-0.002 0.0027 0.3225 -0.9466>}
 deform_index_03_l {parent=deform_index_02_l, loc=<2.5578 -0.123 -0.0252>, rot=<-0.002 0.0067 0.2558 -0.9667>}
 deform_thumb_01_l {parent=deform_hand_l, loc=<1.1806 -0.2643 -2.5309>, rot=<0.6741 -0.3997 0.0268 -0.6205>}
 deform_thumb_02_l {parent=deform_thumb_01_l, loc=<5.9336 -0.0448 0.0401>, rot=<-0.0037 0.0213 0.298 -0.9543>}
 deform_thumb_03_l {parent=deform_thumb_02_l, loc=<2.2223 0.3682 -0.0907>, rot=<0.0118 -0.013 0.0066 -0.9998>}
 deform_lo_arm_twist_01_l {parent=deform_lowerarm_l, loc=<8.4062 0 0>, rot=<0 0 0 1>}
 deform_lo_arm_twist_02_l {parent=deform_lowerarm_l, loc=<13.9825 0 0>, rot=<0 0 0 1>}
 deform_lowerarm_ot_l {parent=deform_lowerarm_l, loc=<0 0 0>, rot=<0.0687 0.0306 0.0067 -0.9971>}
 deform_lowerarm_in_l {parent=deform_lowerarm_l, loc=<0 0 0>, rot=<0.0687 0.0306 0.0067 -0.9971>}
 deform_handSplit_01_l {parent=deform_lowerarm_l, loc=<25.1351 0 0>, rot=<-0.6625 -0.0177 0.1299 -0.7375>}
 deform_handSplit_02_l {parent=deform_handSplit_01_l, loc=<0 0 0>, rot=<0 0 0 1>}
 deform_handStop_l {parent=deform_lowerarm_l, loc=<25.1351 0 0>, rot=<-0.6625 -0.0177 0.1299 -0.7375>}
 deform_lo_arm_twistParent_03_l {parent=deform_lowerarm_l, loc=<19.5588 0 0>, rot=<0 0 0 1>}
 deform_lo_arm_twist_03_l {parent=deform_lo_arm_twistParent_03_l, loc=<0 0 0>, rot=<0 0 0 1>}
 deform_up_arm_twist_02_l {parent=deform_upperarm_l, loc=<9.6764 0 0>, rot=<-0.0146 0.0295 -0.0332 -0.9989>}
 deform_up_arm_twist_04_l {parent=deform_upperarm_l, loc=<22.9496 0 0>, rot=<0 0 0 1>}
 deform_bicep_04_l {parent=deform_up_arm_twist_04_l, loc=<0 0 0>, rot=<-0.0146 0.0295 -0.0332 -0.9989>}
 deform_tricep_04_l {parent=deform_up_arm_twist_04_l, loc=<0 0 0>, rot=<-0.0146 0.0295 -0.0332 -0.9989>}
 deform_up_arm_twist_01_l {parent=deform_upperarm_l, loc=<0 0 0>, rot=<-0.0146 0.0295 -0.0332 -0.9989>}
 deform_up_arm_twist_03_l {parent=deform_upperarm_l, loc=<16.313 0 0>, rot=<-0.0146 0.0295 -0.0332 -0.9989>}
 deform_bicep_03_l {parent=deform_up_arm_twist_03_l, loc=<0 0 0>, rot=<0 0 0 -1>}
 deform_tricep_03_l {parent=deform_up_arm_twist_03_l, loc=<0 0 0>, rot=<0 0 0 -1>}
 deform_clavicle_aim_l {parent=deform_clavicle_l, loc=<0 0 0>, rot=<0 0 0 1>}
 deform_clavicle_r {parent=spine_05, loc=<4.3578 -0.1466 2.3223>, rot=<0.7289 0.0465 -0.6829 0.009>}
 deform_upperarm_r {parent=deform_clavicle_r, loc=<-15.2278 -0.43 -4.2822>, rot=<0.0474 -0.3501 0.0186 -0.9353>}
 deform_lowerarm_r {parent=deform_upperarm_r, loc=<-27.4084 0 -0.0002>, rot=<-0.025 -0.0251 -0.2825 -0.9586>}
 deform_hand_r {parent=deform_lowerarm_r, loc=<-25.1349 -0.0001 -0.0002>, rot=<-0.6625 -0.0177 0.1299 -0.7375>}
 deform_pinky_metacarpal_r {parent=deform_hand_r, loc=<-2.0265 0.7996 -3.2779>, rot=<-0.0381 0.0493 -0.1029 -0.9927>}
 deform_pinky_01_r {parent=deform_pinky_metacarpal_r, loc=<-7.171 0.483 0.1094>, rot=<-0.0654 -0.0035 0.2242 -0.9723>}
 deform_pinky_02_r {parent=deform_pinky_01_r, loc=<-2.8274 -0.0892 0.1043>, rot=<0.0037 0.0011 0.2854 -0.9584>}
 deform_pinky_03_r {parent=deform_pinky_02_r, loc=<-2.2046 -0.0307 0.0078>, rot=<-0.0002 0.0098 0.1445 -0.9895>}
 deform_ring_metacarpal_r {parent=deform_hand_r, loc=<-2.306 0.188 -1.2304>, rot=<-0.0258 0.0406 -0.0802 -0.9956>}
 deform_ring_01_r {parent=deform_ring_metacarpal_r, loc=<-7.0293 0.6196 0.1446>, rot=<-0.0489 -0.0252 0.1892 -0.9804>}
 deform_ring_02_r {parent=deform_ring_01_r, loc=<-3.6173 -0.4646 0.1027>, rot=<0 0 0.2806 -0.9598>}
 deform_ring_03_r {parent=deform_ring_02_r, loc=<-2.4153 -0.1141 -0.0199>, rot=<0 0 0.1189 -0.9929>}
 deform_middle_metacarpal_r {parent=deform_hand_r, loc=<-2.1526 0.0294 0.7288>, rot=<0.0109 -0.0136 -0.068 -0.9975>}
 deform_middle_01_r {parent=deform_middle_metacarpal_r, loc=<-7.3277 0.2449 -0.013>, rot=<-0.0208 0.0042 0.0858 -0.9961>}
 deform_middle_02_r {parent=deform_middle_01_r, loc=<-4.0123 -0.1401 -0.1175>, rot=<-0.0218 0.0258 0.3512 -0.9357>}
 deform_middle_03_r {parent=deform_middle_02_r, loc=<-2.6077 -0.0534 -0.1373>, rot=<0.0082 0.0188 0.1708 -0.9851>}
 deform_index_metacarpal_r {parent=deform_hand_r, loc=<-1.4673 -0.0882 3.0555>, rot=<0.0221 -0.0423 -0.0527 -0.9975>}
 deform_index_01_r {parent=deform_index_metacarpal_r, loc=<-8.3009 -0.0357 0.0524>, rot=<-0.0044 0.0485 0.0924 -0.9945>}
 deform_index_02_r {parent=deform_index_01_r, loc=<-3.3588 0.3004 -0.0452>, rot=<-0.002 0.0027 0.3225 -0.9466>}
 deform_index_03_r {parent=deform_index_02_r, loc=<-2.5578 0.123 0.0252>, rot=<-0.002 0.0067 0.2558 -0.9667>}
 deform_thumb_01_r {parent=deform_hand_r, loc=<-1.1809 0.2645 2.5309>, rot=<0.6741 -0.3997 0.0268 -0.6205>}
 deform_thumb_02_r {parent=deform_thumb_01_r, loc=<-5.9337 0.045 -0.0402>, rot=<-0.0037 0.0213 0.298 -0.9543>}
 deform_thumb_03_r {parent=deform_thumb_02_r, loc=<-2.2222 -0.3682 0.0907>, rot=<0.0118 -0.013 0.0066 -0.9998>}
 deform_lo_arm_twist_01_r {parent=deform_lowerarm_r, loc=<-8.4062 0 -0.0001>, rot=<0 0 0 1>}
 deform_lo_arm_twist_02_r {parent=deform_lowerarm_r, loc=<-13.9829 0.0002 0.0003>, rot=<0 0 0 1>}
 deform_lowerarm_ot_r {parent=deform_lowerarm_r, loc=<0 0 0>, rot=<0.0687 0.0306 0.0067 -0.9971>}
 deform_lowerarm_in_r {parent=deform_lowerarm_r, loc=<0 0 0>, rot=<0.0687 0.0306 0.0067 -0.9971>}
 deform_handSplit_01_r {parent=deform_lowerarm_r, loc=<-25.1349 -0.0001 -0.0002>, rot=<-0.6625 -0.0177 0.1299 -0.7375>}
 deform_handSplit_02_r {parent=deform_handSplit_01_r, loc=<0 0 0>, rot=<0 0 0 1>}
 deform_handStop_r {parent=deform_lowerarm_r, loc=<-25.1349 -0.0001 -0.0002>, rot=<-0.6625 -0.0177 0.1299 -0.7375>}
 deform_lo_arm_twistParent_03_r {parent=deform_lowerarm_r, loc=<-19.5589 0 0.0001>, rot=<0 0 0 1>}
 deform_lo_arm_twist_03_r {parent=deform_lo_arm_twistParent_03_r, loc=<0 0 0>, rot=<0 0 0 1>}
 deform_up_arm_twist_02_r {parent=deform_upperarm_r, loc=<-9.6764 0 -0.0001>, rot=<-0.0146 0.0295 -0.0332 -0.9989>}
 deform_up_arm_twist_03_r {parent=deform_upperarm_r, loc=<-16.3126 0.0001 -0.0004>, rot=<-0.0146 0.0295 -0.0332 -0.9989>}
 deform_bicep_03_r {parent=deform_up_arm_twist_03_r, loc=<0 0 0>, rot=<0 0 0 -1>}
 deform_tricep_03_r {parent=deform_up_arm_twist_03_r, loc=<0 0 0>, rot=<0 0 0 -1>}
 deform_up_arm_twist_04_r {parent=deform_upperarm_r, loc=<-22.9496 0 -0.0001>, rot=<0 0 0 1>}
 deform_bicep_04_r {parent=deform_up_arm_twist_04_r, loc=<0 0 0>, rot=<-0.0146 0.0295 -0.0332 -0.9989>}
 deform_tricep_04_r {parent=deform_up_arm_twist_04_r, loc=<0 0 0>, rot=<-0.0146 0.0295 -0.0332 -0.9989>}
 deform_up_arm_twist_01_r {parent=deform_upperarm_r, loc=<0 0 0>, rot=<-0.0146 0.0295 -0.0332 -0.9989>}
 deform_clavicle_aim_r {parent=deform_clavicle_r, loc=<0 0 0>, rot=<0 0 0 1>}
 lat_l {parent=spine_04, loc=<0 -8.5599 -11.335>, rot=<0 0 0 1>}
 lat_r {parent=spine_04, loc=<0.0004 -8.56 11.335>, rot=<0 0 1 0>}
 belt {parent=spine_01, loc=<0 0 0>, rot=<0 0 0 1>}
 thigh_l {parent=pelvis, loc=<-2.5693 -0.0109 -7.7501>, rot=<0.0012 0.0572 0.0155 -0.9982>}
 calf_l {parent=thigh_l, loc=<-39.0484 0 0>, rot=<0.0006 -0.0161 -0.0391 -0.9991>}
 foot_l {parent=calf_l, loc=<-42.561 0.1077 0.4292>, rot=<0.0418 -0.0455 0.0343 -0.9975>}
 ball_l {parent=foot_l, loc=<-6.1599 13.8264 0.6806>, rot=<0.0001 -0.0001 -0.7063 -0.7079>}
 calf_twist_02_l {parent=calf_l, loc=<-18.8064 -0.0967 0.0397>, rot=<0.0237 0.0095 -0.0035 -0.9997>}
 calf_twist_01_l {parent=calf_l, loc=<-35.813 -0.0967 0.0397>, rot=<0.0237 0.0095 -0.0035 -0.9997>}
 calf_counter_l {parent=calf_l, loc=<0 0 0>, rot=<0 0 0 1>}
 calf_in_l {parent=calf_counter_l, loc=<0 0 0>, rot=<0 0 0 1>}
 calf_kneepad_l {parent=calf_counter_l, loc=<-9.833 4.308 0>, rot=<0 0 0 1>}
 calf_ot_l {parent=calf_l, loc=<0 0 0>, rot=<0 0 0 1>}
 calf_mscl_l {parent=calf_l, loc=<-16.4 0 0>, rot=<0 0 0 1>}
 thigh_twist_01_l {parent=thigh_l, loc=<-21.6303 0 0>, rot=<0.0013 0 0 -1>}
 thigh_counter_l {parent=thigh_l, loc=<-39.0484 0 0>, rot=<0.0006 -0.0161 -0.0391 -0.9991>}
 thigh_r {parent=pelvis, loc=<-2.5693 -0.0109 7.7501>, rot=<0.0572 -0.0012 0.9982 0.0155>}
 calf_r {parent=thigh_r, loc=<39.0484 0 0>, rot=<0.0006 -0.0161 -0.0391 -0.9991>}
 foot_r {parent=calf_r, loc=<42.561 -0.1077 -0.4292>, rot=<0.0418 -0.0455 0.0343 -0.9975>}
 ball_r {parent=foot_r, loc=<6.1599 -13.8264 -0.6806>, rot=<0.0001 -0.0001 -0.7063 -0.7079>}
 calf_twist_02_r {parent=calf_r, loc=<18.8064 0.0967 -0.0397>, rot=<0.0237 0.0095 -0.0035 -0.9997>}
 calf_twist_01_r {parent=calf_r, loc=<35.813 0.0967 -0.0397>, rot=<0.0237 0.0095 -0.0035 -0.9997>}
 calf_counter_r {parent=calf_r, loc=<0 0 0>, rot=<0 0 0 1>}
 calf_in_r {parent=calf_counter_r, loc=<0 0 0>, rot=<0 0 0 1>}
 calf_kneepad_r {parent=calf_counter_r, loc=<9.833 -4.308 0>, rot=<0 0 0 1>}
 calf_ot_r {parent=calf_r, loc=<0 0 0>, rot=<0 0 0 1>}
 calf_mscl_r {parent=calf_r, loc=<16.4 0 0>, rot=<0 0 0 1>}
 thigh_twist_01_r {parent=thigh_r, loc=<21.6303 0 0>, rot=<0.0013 0 0 -1>}
 thigh_counter_r {parent=thigh_r, loc=<39.0484 0 0>, rot=<0.0006 -0.0161 -0.0391 -0.9991>}
 thigh_in_l {parent=pelvis, loc=<-24.0479 0.6538 -10.2194>, rot=<0.0025 0.0571 0.0155 -0.9982>}
 thigh_ot_l {parent=pelvis, loc=<-24.0479 0.6538 -10.2194>, rot=<0.0025 0.0571 0.0155 -0.9982>}
 thigh_in_r {parent=pelvis, loc=<-24.0479 0.6538 10.2194>, rot=<0.0571 -0.0025 0.9982 0.0155>}
 thigh_ot_r {parent=pelvis, loc=<-24.0479 0.6538 10.2194>, rot=<0.0571 -0.0025 0.9982 0.0155>}
]
