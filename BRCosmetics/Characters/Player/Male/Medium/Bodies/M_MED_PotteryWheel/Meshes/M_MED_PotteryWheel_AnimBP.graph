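# Animation Blueprint for the PotteryWheel male medium body cosmetic, retargeting a pose from another mesh and passing it through a deform Control Rig.
# Each update disables an IK flag while the Wacky Wavy emote montage plays.
# It also builds an arm retarget profile with that flag and pole vector offsets derived from upper-arm rotation relative to the clavicles.
# The retarget node shows default profile values, with no visible link to that profile.
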AnimBlueprint /BRCosmetics/Characters/Player/Male/Medium/Bodies/M_MED_PotteryWheel/Meshes/M_MED_PotteryWheel_AnimBP
target_skeleton=&/M_MED_PotteryWheel_Skeleton
graphs [
 AnimGraph {nodes [
   n1 = Control Rig(Source=Retarget Pose From Mesh!(CustomRetargetProfile:Retarget Profile Structure=(RetargetOpProfiles=,bApplyTargetRetargetPose=False,TargetRetargetPoseName="",bApplySourceRetargetPose=False,SourceRetargetPoseName="",bForceAllIKOff=False,bApplyChainSettings=False,ChainSettings=(),bApplyRootSettings=False,RootSettings=(RotationAlpha=1,TranslationAlpha=1,BlendToSource=0,BlendToSourceWeights=(X=1,Y=1,Z=1),ScaleHorizontal=1,ScaleVertical=1,TranslationOffset=(X=0,Y=0,Z=0),RotationOffset=(Pitch=0,Yaw=0,Roll=0),AffectIKHorizontal=1,AffectIKVertical=0),bApplyGlobalSettings=False,GlobalSettings=(bEnableRoot=True,bEnableFK=True,bEnableIK=True,bEnablePost=True,bCopyBasePose=False,CopyBasePoseRoot="",SourceScaleFactor=1,bWarping=False,DirectionSource=Goals,ForwardDirection=Y,DirectionChain="",WarpForwards=1,SidewaysOffset=0,WarpSplay=1))) -> (Pose:Pose Link Structure) {
     is=title AnimGraphNode_RetargetPoseFromMesh
     iKRetargeterAsset=&/M_MED_PotterWheel_Retargeter2
    }) {
    alphaBoolBlend {blendOption=Linear}
    controlRigAssetReference {blueprintRigClass=&/M_MED_PotteryWheelDeform2_CtrlRig.M_MED_PotteryWheelDeform2_CtrlRig_C}
   }
   Output Pose(Result=@n1) {layerGroup=None}
   Set members in Target Chain FKSettings!(
    execute:Exec
    StructRef:"Target Chain FKSettings Structure (by ref)"=Break Target Chain Settings!(TargetChainSettings:"Target Chain Settings Structure (by ref)"=GetChainSettingsFromRetargetAsset!(
      RetargetAsset:IKRetargeter Object Reference=&/M_MED_PotterWheel_Retargeter
      TargetChainName:Name=LeftArm
      OptionalProfileName:Name=LARM
     ) -> (ReturnValue:Target Chain Settings Structure)) -> (FK:Target Chain FKSettings Structure, IK:Target Chain IKSettings Structure, SpeedPlanting:Target Chain Speed Plant Settings Structure) {is=title K2Node_BreakStruct}
    PoleVectorOffset:float=0
   ) -> (then:Exec, StructOut:"Target Chain FKSettings Structure (by ref)") {is=title K2Node_SetFieldsInStruct}
  ]}
 EventGraph {nodes [
   Event BlueprintUpdateAnimation()
   > n1 = Sequence()
   n2 = GetAnimInstance(self=GetOwnerSkeletalMeshComponent())
   > Is Valid(InputObject=@n2)
   > n3 = Set isCustomEmotePlaying(isCustomEmotePlaying:double=SelectFloat(A=1, B=0, bPickA=Montage_IsPlaying(self=@n2, Montage=/BRCosmetics/Animation/Game/MainPlayer/Emotes/Wacky_Wavy/Emote_WackWavy_CMM_M))) -> (Output_Get:double)
   > n4 = Branch(Condition="To Boolean (Integer)"!(InInt:int=Truncate(A=@n3)) -> (ReturnValue:bool))
   > Set enableIK(enableIK:bool=false) -> (Output_Get:bool)
   Set enableIK(execute=@n4.else, enableIK=true)
   n5 = GetOwnerSkeletalMeshComponent()
   n6 = Set armDn_l(
    execute=@n1.then_1
    armDn_l:float=float * float(
     A=GetAlphaFromReferenceTransformDelta(SkeletalMeshComponent=@n5, BoneName=upperarm_l, BaseName=clavicle_l, DeltaComponentType=EulerY, InputRange=(X=0,Y=-40), TwistAxis="1,0,0")
     B=-20
    ) {is=title K2Node_PromotableOperator}
   ) -> (Output_Get:float)
   > n7 = Set ArmDn_r(ArmDn_r:float=float * float(
     A=GetAlphaFromReferenceTransformDelta(SkeletalMeshComponent=@n5, BoneName=upperarm_r, BaseName=clavicle_r, DeltaComponentType=EulerY, InputRange=(X=0,Y=-40), TwistAxis="1,0,0")
     B=20
    )) -> (Output_Get:float)
   n8 = Make Target Chain Settings!(
    FK:Target Chain FKSettings Structure=Make Target Chain FKSettings!(
     EnableFK:bool=True
     RotationMode:ERetargetRotationMode Enum=OneToOne
     RotationAlpha:float=1
     TranslationMode:ERetargetTranslationMode Enum=None
     TranslationAlpha:float=1
     PoleVectorMatching:float=0
     PoleVectorMaintainOffset:bool=False
     PoleVectorOffset:float=@n6
    ) -> (TargetChainFKSettings:Target Chain FKSettings Structure) {is=title K2Node_MakeStruct}
    IK:Target Chain IKSettings Structure=Make Target Chain IKSettings!(
     EnableIK:bool=True
     BlendToSource:float=1
     BlendToSourceTranslation:float=1
     BlendToSourceRotation:float=0
     BlendToSourceWeights:Vector="1,1,1"
     StaticOffset:Vector="0,0,0"
     StaticLocalOffset:Vector="0,0,0"
     StaticRotationOffset:Rotator="0,0,0"
     ScaleVertical:float=1
     Extension:float=1
     bAffectedByIKWarping:bool=True
    ) -> (TargetChainIKSettings:Target Chain IKSettings Structure) {is=title K2Node_MakeStruct}
    SpeedPlanting:Target Chain Speed Plant Settings Structure=(EnableSpeedPlanting=False,SpeedCurveName="",SpeedThreshold=15,UnplantStiffness=250,UnplantCriticalDamping=1)
   ) -> (TargetChainSettings:Target Chain Settings Structure) {is=title K2Node_MakeStruct}
   n9 = Make Target Chain Settings(
    FK=Make Target Chain FKSettings(
     EnableFK=True
     RotationMode=OneToOne
     RotationAlpha=1
     TranslationMode=None
     TranslationAlpha=1
     PoleVectorMatching=0
     PoleVectorMaintainOffset=False
     PoleVectorOffset=float + float!(A=@n7, B, C:double) -> (ReturnValue)
    )
    IK=Make Target Chain IKSettings(
     EnableIK=True
     BlendToSource=1
     BlendToSourceTranslation=1
     BlendToSourceRotation=0
     BlendToSourceWeights="1,1,1"
     StaticOffset="0,0,0"
     StaticLocalOffset="0,0,0"
     StaticRotationOffset="0,0,0"
     ScaleVertical=1
     Extension=1
     bAffectedByIKWarping=True
    )
    SpeedPlanting=(EnableSpeedPlanting=False,SpeedCurveName="",SpeedThreshold=15,UnplantStiffness=250,UnplantCriticalDamping=1)
   )
   n10 = Get enableIK() -> (enableIK:bool)
   > Set Retarget Profile Arms(Retarget Profile Arms:Retarget Profile Structure=Make Retarget Profile!(
     RetargetOpProfiles:Array of Retarget Op Profile Structures
     bApplyTargetRetargetPose:bool=False
     TargetRetargetPoseName:Name=None
     bApplySourceRetargetPose:bool=False
     SourceRetargetPoseName:Name=None
     bForceAllIKOff:bool=False
     bApplyChainSettings:bool=true
     ChainSettings:Map of Names to Target Chain Settings Structures=Make Map!(Key 0:Name=LeftArm, Value 0:Target Chain Settings Structure=@n8, Key 1:Name=RightArm, Value 1:Target Chain Settings Structure=@n9) -> (Map:Map of Names to Target Chain Settings Structures) {is=title K2Node_MakeMap}
     bApplyRootSettings:bool=False
     RootSettings:Target Root Settings Structure=(RotationAlpha=1,TranslationAlpha=1,BlendToSource=0,BlendToSourceWeights=(X=1,Y=1,Z=1),ScaleHorizontal=1,ScaleVertical=1,TranslationOffset=(X=0,Y=0,Z=0),RotationOffset=(Pitch=0,Yaw=0,Roll=0),AffectIKHorizontal=1,AffectIKVertical=0)
     bApplyGlobalSettings:bool=true
     GlobalSettings:Retarget Global Settings Structure=Make Retarget Global Settings!(
      bEnableRoot:bool=True
      bEnableFK:bool=True
      bEnableIK:bool=@n10
      bEnablePost:bool=True
      bCopyBasePose:bool=False
      CopyBasePoseRoot:Name=None
      SourceScaleFactor:float=1
      bWarping:bool=False
      DirectionChain:Name=None
      WarpForwards:float=1
      SidewaysOffset:float=0
      WarpSplay:float=1
     ) -> (RetargetGlobalSettings:Retarget Global Settings Structure) {is=title K2Node_MakeStruct}
    ) -> (RetargetProfile:Retarget Profile Structure) {is=title K2Node_MakeStruct}) -> (Output_Get:Retarget Profile Structure)
   Set armUp_l(armUp_l:double=float * float(
     A=GetAlphaFromReferenceTransformDelta(BoneName=upperarm_l, BaseName=clavicle_l, DeltaComponentType=EulerY, InputRange=(X=-20,Y=0), TwistAxis="1,0,0")
     B=20
    )) -> (Output_Get:double)
  ]}
]
variables [
 armDn_l {type=number}
 ArmDn_r {type=number}
 armUp_l {type=number}
 armUp_r {type=number}
 ArmUpExt_r {type=number}
 body Mesh {type=SkeletalMeshComponent}
 Custom Retarget Profile {
  type=RetargetProfile
  default {
   bApplyChainSettings=false
   bApplyGlobalSettings=false
   bApplyRootSettings=false
   bApplySourceRetargetPose=false
   bApplyTargetRetargetPose=false
   bForceAllIKOff=false
   sourceRetargetPoseName=None
   targetRetargetPoseName=None
   chainSettings {}
   globalSettings {
    bCopyBasePose=false
    bEnableFK=true
    bEnableIK=true
    bEnablePost=true
    bEnableRoot=true
    bWarping=false
    copyBasePoseRoot=None
    directionChain=None
    directionSource=Goals
    forwardDirection=Y
    sidewaysOffset=0
    sourceScaleFactor=1
    warpForwards=1
    warpSplay=1
   }
   retargetOpProfiles []
   rootSettings {
    affectIKHorizontal=1
    affectIKVertical=0
    blendToSource=0
    rotationAlpha=1
    scaleHorizontal=1
    scaleVertical=1
    translationAlpha=1
    blendToSourceWeights=<1 1 1>
    rotationOffset=<rot 0 0 0>
    translationOffset=<0 0 0>
   }
  }
 }
 enableIK {type=boolean}
 isCustomEmotePlaying {type=number}
 Pole Vector Offset LARM {type=number}
 Retarget Profile Arms {
  type=RetargetProfile
  default {
   bApplyChainSettings=false
   bApplyGlobalSettings=false
   bApplyRootSettings=false
   bApplySourceRetargetPose=false
   bApplyTargetRetargetPose=false
   bForceAllIKOff=false
   sourceRetargetPoseName=None
   targetRetargetPoseName=None
   chainSettings {}
   globalSettings {
    bCopyBasePose=false
    bEnableFK=true
    bEnableIK=true
    bEnablePost=true
    bEnableRoot=true
    bWarping=false
    copyBasePoseRoot=None
    directionChain=None
    directionSource=Goals
    forwardDirection=Y
    sidewaysOffset=0
    sourceScaleFactor=1
    warpForwards=1
    warpSplay=1
   }
   retargetOpProfiles []
   rootSettings {
    affectIKHorizontal=1
    affectIKVertical=0
    blendToSource=0
    rotationAlpha=1
    scaleHorizontal=1
    scaleVertical=1
    translationAlpha=1
    blendToSourceWeights=<1 1 1>
    rotationOffset=<rot 0 0 0>
    translationOffset=<0 0 0>
   }
  }
 }
]
